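# Control Rig for the F_MED_CheddarBandit body that adds corrective bone deformation to the character's outfit and body.
# The shared DeformRig function runs first, then spherical pose readers weight transform offsets on foot, holster, belt and strap bones.
# Calf rotation drives the right knee pad, and pelvis height drives the left glute.
# Holster offsets also follow weapon and targeting flags, with interpolated weights that fade during emotes and zero in frontend.
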ControlRigBlueprint /BRCosmetics/Characters/Player/Female/Medium/Bodies/F_MED_CheddarBandit/Meshes/F_MED_CheddarBandit_CtrlRig
graphs [
 ContainedGraph {nodes [
   n1 = RigVMFunctionEntryNode()
   n2 = &RigUnit_GetTransform(Space=LocalSpace, Item=@n1.Item)
   n3 = &RigUnit_GetTransform(Space=LocalSpace, bInitial=true, Item=@n1.Item)
   > RigVMFunctionReturnNode(OutTransform=/Script/RigVM.RigVMFunction_MathTransformToEulerTransform(Value=&RigVMFunction_MathTransformMakeRelative(Global=@n2.Transform, Parent=@n3.Transform).Local) {is=struct RigVMUnitNode, template="Cast::Execute(in Value,out Result)"}.Result)
  ]}
 RigVMFunctionLibrary {nodes [DeformRig_GetDeltaRefTransform_FNC = RigVMCollapseNode(ExecuteContext=(), Item=(Type=Bone))]}
 RigVMFunction_Sequence_2_ContainedGraph {nodes [
   n1 = RigVMFunctionEntryNode()
   RigVMFunction_Sequence = &RigVMFunction_Sequence(ExecuteContext=@n1.ExecuteContext)
   n2 = &RigVMFunction_Sequence(ExecuteContext=@RigVMFunction_Sequence.B)
   n3 = &RigVMFunction_Sequence(ExecuteContext=@n2.B)
   n4 = &RigVMFunction_Sequence(ExecuteContext=@n3.B)
   n5 = &RigVMFunction_Sequence(ExecuteContext=@n4.B)
   RigVMFunctionReturnNode(A=@RigVMFunction_Sequence.A, B=@n2.A, C=@n3.A, D=@n4.A, E=@n5.A, F=@n5.B)
  ]}
 RigVMModel {nodes [
   RigUnit_BeginExecution = &RigUnit_BeginExecution()
   n1 = &RigVMFunction_Sequence(ExecuteContext=@RigUnit_BeginExecution.ExecutePin)
   n2 = &RigVMFunction_ControlFlowBranch(ExecuteContext=@n1.A, Condition=RigVMVariableNode() {variable=isFrontend}.Value)
   n3 = &RigVMFunction_AlphaInterp(
    InterpSpeedDecreasing=5
    InterpSpeedIncreasing=5
    Scale=1
    bClampResult=False
    Value=RigVMDispatchNode(False=1, True=0, Condition=RigVMVariableNode() {variable=is Playing Emote}.Value) {template="DISPATCH_RigVMDispatch_If(in Condition,in True,in False,out Result)"}.Result
   )
   RigVMVariableNode(Value=@n3.Result, ExecuteContext=@n2.False) {variable=isNotPlayingEmoteFloat}
   RigVMVariableNode(Value=0, ExecuteContext=@n2.True) {variable=isNotPlayingEmoteFloat}
   FN_DeformRig_FNC = RigVMFunctionReferenceNode(
    DeformRig_ Curve="(deformRig_l_upperarm_up_alpha_72_#1=False,deformRig_r_upperarm_up_alpha_73_#2=False,deformRig_l_upperarm_fr_alpha_74_#3=False,deformRig_r_upperarm_fr_alpha_75_CB51B8 ...(truncated)"
    DeformRig_Bool="(upperarm_1_#4=True,elbow_3_#5=True,wrist_7_#6=True,groin_8_#7=True,glute_17_#8=True,knee_9_40C408324CA261 ...(truncated)"
    DeformRig_Debug=(upperarm_debug_18_#4=False,wrist_debug_20_#6=False,lat_debug_26_#9=False,groinGlute_debug_27_#7=False)
    DeformRig_Value="(upperarm_up_18_#10=(Rotation=(X=0,Y=0,Z=0,W=1),Translation=(X=0,Y=0,Z=8),Scale3D=(X=0.7,Y=1,Z=1)),upperarm_fwd_fr_19_#11=(Rotation=(X=0,Y=0,Z=0,W=1),Translation=(X=0,Y=10,Z=0),Scale3D=(X=0.5,Y=1,Z= ...(truncated)"
    ExecuteContext=@n1.B
   )
   > RigVMFunction_Sequence = RigVMAggregateNode()
   DeformRig_GetDeltaRefTransform_FNC_1 = RigVMFunctionReferenceNode(Item=(Type=Bone,Name="calf_r"), ExecuteContext=@RigVMFunction_Sequence.A)
   > &RigUnit_ModifyTransforms(
    ItemToModify=((Item=(Type=Bone,Name="def_kneePad_r"),Transform=(Rotation=(X=0,Y=0,Z=0.3903,W=0.9207),Translation=(X=0.0921,Y=6.2962,Z=0),Scale3D=(X=0.667,Y=1,Z=0.667))))
    WeightMinimum=0
    Weight=&RigVMFunction_MathFloatRemap(SourceMaximum=100, SourceMinimum=10, bClamp=true, Value=@DeformRig_GetDeltaRefTransform_FNC_1.OutTransform.Rotation.Yaw).Result
   )
   n4 = &RigUnit_SphericalPoseReader(
    ActiveRegionScaleFactors=(PositiveWidth=1,NegativeWidth=0.5,PositiveHeight=1,NegativeHeight=1)
    ActiveRegionSize=0.1
    Debug=(bDrawDebug=false,bDraw2D=False,bDrawLocalAxes=False,DebugScale=25,DebugSegments=20,DebugThickness=0.25)
    DriverAxis=(X=0,Y=1,Z=0)
    DriverItem=(Type=Bone,Name="foot_l")
    FalloffRegionScaleFactors=(PositiveWidth=1,NegativeWidth=1,PositiveHeight=1,NegativeHeight=1)
    FalloffSize=0.2
    FlipHeightScaling=False
    FlipWidthScaling=False
    OptionalParentItem=(Type=Bone,Name="None")
    RotationOffset=(X=90,Y=0,Z=-38)
    ExecutePin=@RigVMFunction_Sequence.B
   )
   > &RigUnit_ModifyTransforms(
    ItemToModify="((Item=(Type=Bone,Name=\"def_foot_fr_01_l\"),Transform=(Rotation=(X=0,Y=0.1325,Z=0,W=0.9912),Translation=(X=0.5059,Y=0,Z=-2.2377),Scale3D=(X=1,Y=1,Z=1.7787))),(Item=(Type=Bone,Name=\"def_foot_bk_01_l\"),Transform=(Rotation=(X=0,Y=0,Z=0,W=-1),Translation=(X=1.3 ...(truncated)"
    WeightMinimum=0
    Weight=@n4.OutputParam
   )
   > n5 = &RigUnit_SphericalPoseReader(
    ActiveRegionScaleFactors=(PositiveWidth=1,NegativeWidth=0.5,PositiveHeight=1,NegativeHeight=1)
    ActiveRegionSize=0.1
    Debug=(bDrawDebug=false,bDraw2D=False,bDrawLocalAxes=False,DebugScale=25,DebugSegments=20,DebugThickness=0.25)
    DriverAxis=(X=0,Y=1,Z=0)
    DriverItem=(Type=Bone,Name="foot_r")
    FalloffRegionScaleFactors=(PositiveWidth=1,NegativeWidth=1,PositiveHeight=1,NegativeHeight=1)
    FalloffSize=0.2
    FlipHeightScaling=False
    FlipWidthScaling=False
    OptionalParentItem=(Type=Bone,Name="None")
    RotationOffset=(X=90,Y=0,Z=-38)
   )
   > &RigUnit_ModifyTransforms(
    ItemToModify="((Item=(Type=Bone,Name=\"def_foot_fr_01_r\"),Transform=(Rotation=(X=0,Y=0.1325,Z=0,W=0.9912),Translation=(X=-0.0088,Y=0,Z=-2.23),Scale3D=(X=1,Y=1,Z=1.7787))),(Item=(Type=Bone,Name=\"def_foot_bk_01_r\"),Transform=(Rotation=(X=0,Y=0,Z=0,W=-1),Translation=(X=1.30 ...(truncated)"
    WeightMinimum=0
    Weight=@n5.OutputParam
   )
   > n6 = &RigUnit_SphericalPoseReader(
    ActiveRegionScaleFactors=(PositiveWidth=1,NegativeWidth=1,PositiveHeight=1.6,NegativeHeight=1.6)
    ActiveRegionSize=0.12
    Debug=(bDrawDebug=false,bDraw2D=False,bDrawLocalAxes=False,DebugScale=25,DebugSegments=20,DebugThickness=0.25)
    DriverAxis=(X=1,Y=0,Z=0)
    DriverItem=(Type=Bone,Name="foot_l")
    FalloffRegionScaleFactors=(PositiveWidth=3,NegativeWidth=0.1,PositiveHeight=2,NegativeHeight=2)
    FalloffSize=0.3
    FlipHeightScaling=False
    FlipWidthScaling=False
    OptionalParentItem=(Type=Bone,Name="None")
    RotationOffset=(X=90,Y=0,Z=-5)
   )
   > &RigUnit_ModifyTransforms(
    ItemToModify="((Item=(Type=Bone,Name=\"def_foot_fr_01_l\"),Transform=(Rotation=(X=0,Y=0.4554,Z=0,W=-0.8903),Translation=(X=-3.6905,Y=0,Z=3.5021),Scale3D=(X=1,Y=1,Z=1))),(Item=(Type=Bone,Name=\"def_foot_bk_01_l\"),Transform=(Rotation=(X=0,Y=0.4226,Z=0,W=0.9063),Translation=( ...(truncated)"
    WeightMinimum=0
    Weight=@n6.OutputParam
   )
   > n7 = &RigUnit_SphericalPoseReader(
    ActiveRegionScaleFactors=(PositiveWidth=1,NegativeWidth=1,PositiveHeight=1.6,NegativeHeight=1.6)
    ActiveRegionSize=0.12
    Debug=(bDrawDebug=false,bDraw2D=False,bDrawLocalAxes=False,DebugScale=25,DebugSegments=20,DebugThickness=0.25)
    DriverAxis=(X=1,Y=0,Z=0)
    DriverItem=(Type=Bone,Name="foot_r")
    FalloffRegionScaleFactors=(PositiveWidth=3,NegativeWidth=0.1,PositiveHeight=2,NegativeHeight=2)
    FalloffSize=0.3
    FlipHeightScaling=False
    FlipWidthScaling=False
    OptionalParentItem=(Type=Bone,Name="None")
    RotationOffset=(X=90,Y=0,Z=-5)
   )
   > &RigUnit_ModifyTransforms(
    ItemToModify="((Item=(Type=Bone,Name=\"def_foot_fr_01_r\"),Transform=(Rotation=(X=0,Y=0.4554,Z=0,W=-0.8903),Translation=(X=-3.6905,Y=0,Z=3.5021),Scale3D=(X=1,Y=1,Z=1))),(Item=(Type=Bone,Name=\"def_foot_bk_01_r\"),Transform=(Rotation=(X=0,Y=0.4226,Z=0,W=0.9063),Translation=( ...(truncated)"
    WeightMinimum=0
    Weight=@n7.OutputParam
   )
   n8 = &RigVMFunction_Sequence(ExecuteContext=@RigVMFunction_Sequence.C)
   n9 = &RigUnit_SphericalPoseReader(
    ActiveRegionScaleFactors=(PositiveWidth=1,NegativeWidth=1,PositiveHeight=1,NegativeHeight=1)
    ActiveRegionSize=0.1
    Debug=(bDrawDebug=false,bDraw2D=False,bDrawLocalAxes=False,DebugScale=25,DebugSegments=20,DebugThickness=0.25)
    DriverAxis=(X=-1,Y=0,Z=0)
    DriverItem=(Type=Bone,Name="upperarm_r")
    FalloffRegionScaleFactors=(PositiveWidth=1,NegativeWidth=2,PositiveHeight=1,NegativeHeight=1)
    FalloffSize=0.2
    FlipHeightScaling=False
    FlipWidthScaling=False
    OptionalParentItem=(Type=Bone,Name="spine_04")
    RotationOffset=(X=0,Y=33.4335,Z=0)
    ExecutePin=@n8.A
   )
   AlphaInterp = &RigVMFunction_AlphaInterp(InterpSpeedDecreasing=10, InterpSpeedIncreasing=10, Scale=1, bClampResult=true, bMapRange=false, Value=@n9.OutputParam)
   Multiply = &RigVMFunction_MathDoubleMul(A=@AlphaInterp.Result, B=RigVMVariableNode() {variable=isNotPlayingEmoteFloat}.Value)
   > &RigUnit_ModifyTransforms(
    ItemToModify=((Item=(Type=Bone,Name="Holster_r"),Transform=(Rotation=(X=-0.3989,Y=0.0322,Z=0.0706,W=-0.9137),Translation=(X=1.4837,Y=-3.0168,Z=0.2672),Scale3D=(X=1,Y=1,Z=1))))
    WeightMinimum=0
    Weight=@Multiply.Result
   )
   > n10 = &RigUnit_SphericalPoseReader(
    ActiveRegionScaleFactors=(PositiveWidth=1,NegativeWidth=1,PositiveHeight=1,NegativeHeight=1)
    ActiveRegionSize=0.1
    Debug=(bDrawDebug=false,bDraw2D=False,bDrawLocalAxes=False,DebugScale=25,DebugSegments=20,DebugThickness=0.25)
    DriverAxis=(X=-1,Y=0,Z=0)
    DriverItem=(Type=Bone,Name="upperarm_l")
    FalloffRegionScaleFactors=(PositiveWidth=1,NegativeWidth=2,PositiveHeight=1,NegativeHeight=1)
    FalloffSize=0.2
    FlipHeightScaling=False
    FlipWidthScaling=False
    OptionalParentItem=(Type=Bone,Name="spine_04")
    RotationOffset=(X=0,Y=33.4335,Z=0)
   )
   n11 = &RigVMFunction_AlphaInterp(InterpSpeedDecreasing=10, InterpSpeedIncreasing=10, Scale=1, bClampResult=true, bMapRange=false, Value=@n10.OutputParam)
   > &RigUnit_ModifyTransforms(
    ItemToModify=((Item=(Type=Bone,Name="Holster_L"),Transform=(Rotation=(X=0.3989,Y=-0.0314,Z=-0.0719,W=0.9136),Translation=(X=-1.5,Y=3,Z=0.26),Scale3D=(X=1,Y=1,Z=1))))
    WeightMinimum=0
    Weight=&RigVMFunction_MathDoubleMul(A=@n11.Result, B=RigVMVariableNode() {variable=isNotPlayingEmoteFloat}.Value).Result
   )
   RigVMFunction_MathBoolAnd = &RigVMFunction_MathBoolAnd(B=RigVMVariableNode() {variable=isTargeting}.Value, A=RigVMVariableNode() {variable=isAssaultRifle}.Value)
   RigVMFunction_MathBoolToFloat = &RigVMFunction_MathBoolToFloat(Value=@RigVMFunction_MathBoolAnd.Result)
   n12 = &RigVMFunction_AlphaInterp(InterpSpeedDecreasing=10, InterpSpeedIncreasing=10, Scale=1, bClampResult=true, bMapRange=false, Value=@RigVMFunction_MathBoolToFloat.Result)
   &RigUnit_ModifyTransforms(
    ItemToModify="((Item=(Type=Bone,Name=\"Holster_r\"),Transform=(Rotation=(X=-0.001,Y=-0.218,Z=0.0283,W=0.9755),Translation=(X=1.899,Y=-2.0483,Z=2.3665),Scale3D=(X=1,Y=1,Z=1))),(Item=(Type=Bone,Name=\"Holster_L\"),Transform=(Rotation=(X=0,Y=0,Z=0,W=1),Translation=(X=0,Y=0,Z=0 ...(truncated)"
    WeightMinimum=0
    Weight=&RigVMFunction_MathDoubleMul(A=@n12.Result, B=RigVMVariableNode() {variable=isNotPlayingEmoteFloat}.Value).Result
    ExecutePin=@n8.B
   )
   RigVMFunction_MathBoolNot = &RigVMFunction_MathBoolNot(Value=RigVMVariableNode() {variable=isTargeting}.Value)
   n13 = RigVMVariableNode() {variable=isPistol}
   n14 = &RigVMFunction_AlphaInterp(
    InterpSpeedDecreasing=10
    InterpSpeedIncreasing=10
    Scale=1
    bClampResult=true
    bMapRange=false
    Value=&RigVMFunction_MathBoolToFloat(Value=&RigVMFunction_MathBoolAnd(B=@RigVMFunction_MathBoolNot.Result, A=@n13.Value).Result).Result
   )
   > &RigUnit_ModifyTransforms(
    ItemToModify="((Item=(Type=Bone,Name=\"Holster_r\"),Transform=(Rotation=(X=0.0286,Y=-0.2099,Z=0.034,W=0.9767),Translation=(X=2.0097,Y=1.9252,Z=0.5826),Scale3D=(X=1,Y=1,Z=1))),(Item=(Type=Bone,Name=\"Holster_L\"),Transform=(Rotation=(X=-0.0081,Y=-0.1646,Z=0.0486,W=0.9851),Tr ...(truncated)"
    WeightMinimum=0
    Weight=&RigVMFunction_MathDoubleMul(A=@n14.Result, B=RigVMVariableNode() {variable=isNotPlayingEmoteFloat}.Value).Result
   )
   n15 = RigVMVariableNode() {variable=isSMG}
   n16 = RigVMVariableNode() {variable=isTargeting}
   n17 = &RigVMFunction_AlphaInterp(
    InterpSpeedDecreasing=10
    InterpSpeedIncreasing=10
    Scale=1
    bClampResult=true
    bMapRange=false
    Value=&RigVMFunction_MathBoolToFloat(Value=&RigVMFunction_MathBoolAnd(A=@n15.Value, B=@n16.Value).Result).Result
   )
   > &RigUnit_ModifyTransforms(
    ItemToModify="((Item=(Type=Bone,Name=\"Holster_r\"),Transform=(Rotation=(X=0.1573,Y=-0.1471,Z=0.1218,W=0.9689),Translation=(X=2.0933,Y=-1.3169,Z=1.648),Scale3D=(X=1,Y=1,Z=1))),(Item=(Type=Bone,Name=\"Holster_L\"),Transform=(Rotation=(X=0,Y=0,Z=0,W=1),Translation=(X=0,Y=0,Z= ...(truncated)"
    WeightMinimum=0
    Weight=&RigVMFunction_MathDoubleMul(A=@n17.Result, B=RigVMVariableNode() {variable=isNotPlayingEmoteFloat}.Value).Result
   )
   n18 = &RigUnit_SphericalPoseReader(
    ActiveRegionScaleFactors=(PositiveWidth=3.5,NegativeWidth=0.5,PositiveHeight=3,NegativeHeight=3)
    ActiveRegionSize=0.1
    Debug=(bDrawDebug=false,DebugScale=25,DebugSegments=20,DebugThickness=0.25)
    DriverAxis=(X=-1,Y=0,Z=0)
    DriverItem=(Type=Bone,Name="thigh_l")
    FalloffRegionScaleFactors=(PositiveWidth=1,NegativeWidth=1,PositiveHeight=1,NegativeHeight=1)
    FalloffSize=0.45
    OptionalParentItem=(Type=Bone)
    RotationOffset=(X=90,Y=0,Z=0)
    ExecutePin=@RigVMFunction_Sequence.D
   )
   > &RigUnit_ModifyTransforms(
    ItemToModify="((Item=(Type=Bone,Name=\"dyn_strapFront_01_l\"),Transform=(Rotation=(X=-0.3102,Y=0.292,Z=0.2934,W=-0.8558),Translation=(X=-0.5576,Y=0.3804,Z=-1.76),Scale3D=(X=1,Y=1,Z=1))),(Item=(Type=Bone,Name=\"dyn_beltCorr_l\"),Transform=(Rotation=(X=0,Y=0,Z=0,W=-1),Transla ...(truncated)"
    WeightMinimum=0
    Weight=@n18.OutputParam
   )
   > n19 = &RigUnit_SphericalPoseReader(
    ActiveRegionScaleFactors=(PositiveWidth=3.7,NegativeWidth=0.5,PositiveHeight=3,NegativeHeight=3)
    ActiveRegionSize=0.1
    Debug=(bDrawDebug=false,DebugScale=25,DebugSegments=20,DebugThickness=0.25)
    DriverAxis=(X=1,Y=0,Z=0)
    DriverItem=(Type=Bone,Name="thigh_r")
    FalloffRegionScaleFactors=(PositiveWidth=1,NegativeWidth=1.2,PositiveHeight=1,NegativeHeight=1)
    FalloffSize=0.45
    FlipHeightScaling=false
    FlipWidthScaling=true
    OptionalParentItem=(Type=Bone)
    RotationOffset=(X=-90,Y=0,Z=0)
   )
   > &RigUnit_ModifyTransforms(
    ItemToModify="((Item=(Type=Bone,Name=\"dyn_strapFront_01_r\"),Transform=(Rotation=(X=-0.3173,Y=-0.0095,Z=0.6524,W=0.6882),Translation=(X=-6.1896,Y=-0.0814,Z=-1.1848),Scale3D=(X=1,Y=1,Z=1))),(Item=(Type=Bone,Name=\"dyn_beltCorr_r\"),Transform=(Rotation=(X=0,Y=0,Z=0,W=1),Tran ...(truncated)"
    WeightMinimum=0
    Weight=@n19.OutputParam
   )
   > n20 = &RigUnit_SphericalPoseReader(
    ActiveRegionScaleFactors=(PositiveWidth=1,NegativeWidth=1,PositiveHeight=1,NegativeHeight=1)
    ActiveRegionSize=0.1
    Debug=(bDrawDebug=false,bDraw2D=False,bDrawLocalAxes=False,DebugScale=25,DebugSegments=20,DebugThickness=0.25)
    DriverAxis=(X=1,Y=0,Z=0)
    DriverItem=(Type=Bone,Name="thigh_r")
    FalloffRegionScaleFactors=(PositiveWidth=2.7,NegativeWidth=1,PositiveHeight=1,NegativeHeight=1)
    FalloffSize=0.2
    FlipHeightScaling=False
    FlipWidthScaling=False
    OptionalParentItem=(Type=Bone,Name="None")
    RotationOffset=(X=-25,Y=0,Z=0)
   )
   > &RigUnit_ModifyTransforms(
    ItemToModify="((Item=(Type=Bone,Name=\"dyn_strapFront_01_r\"),Transform=(Rotation=(X=0,Y=0,Z=-0.6003,W=-0.7998),Translation=(X=0,Y=0,Z=0),Scale3D=(X=1,Y=1,Z=1))),(Item=(Type=Bone,Name=\"dyn_beltCorr_r\"),Transform=(Rotation=(X=-0.0504,Y=0,Z=0,W=-0.9987),Translation=(X=0,Y=- ...(truncated)"
    WeightMinimum=0
    Weight=@n20.OutputParam
   )
   > n21 = &RigUnit_SphericalPoseReader(
    ActiveRegionScaleFactors=(PositiveWidth=1,NegativeWidth=1,PositiveHeight=1,NegativeHeight=1)
    ActiveRegionSize=0.1
    Debug=(bDrawDebug=true,bDraw2D=False,bDrawLocalAxes=False,DebugScale=25,DebugSegments=20,DebugThickness=0.25)
    DriverAxis=(X=-1,Y=0,Z=0)
    DriverItem=(Type=Bone,Name="thigh_l")
    FalloffRegionScaleFactors=(PositiveWidth=5,NegativeWidth=1,PositiveHeight=1,NegativeHeight=1)
    FalloffSize=0.2
    FlipHeightScaling=False
    FlipWidthScaling=False
    OptionalParentItem=(Type=Bone,Name="None")
    RotationOffset=(X=205,Y=0,Z=180)
   )
   > &RigUnit_ModifyTransforms(
    ItemToModify="((Item=(Type=Bone,Name=\"dyn_strapFront_01_l\"),Transform=(Rotation=(X=0,Y=-0.6916,Z=0,W=0.7223),Translation=(X=0,Y=0,Z=0),Scale3D=(X=1,Y=1,Z=1))),(Item=(Type=Bone,Name=\"dyn_beltCorr_l\"),Transform=(Rotation=(X=-0.0353,Y=0,Z=0,W=0.9994),Translation=(X=0,Y=0,Z ...(truncated)"
    WeightMinimum=0
    Weight=@n21.OutputParam
   )
   n22 = &RigUnit_SphericalPoseReader(
    ActiveRegionScaleFactors=(PositiveWidth=1,NegativeWidth=1,PositiveHeight=1,NegativeHeight=1)
    ActiveRegionSize=0.1
    Debug=(bDrawDebug=false,bDraw2D=False,bDrawLocalAxes=False,DebugScale=25,DebugSegments=20,DebugThickness=0.25)
    DriverAxis=(X=1,Y=0,Z=0)
    DriverItem=(Type=Bone,Name="spine_03")
    FalloffRegionScaleFactors=(PositiveWidth=3.5,NegativeWidth=1,PositiveHeight=2,NegativeHeight=2)
    FalloffSize=0.2
    FlipHeightScaling=False
    FlipWidthScaling=False
    OptionalParentItem=(Type=Bone,Name="spine_01")
    RotationOffset=(X=90,Y=0,Z=0)
    ExecutePin=@RigVMFunction_Sequence.E
   )
   > &RigUnit_ModifyTransforms(
    ItemToModify="((Item=(Type=Bone,Name=\"dyn_StomachStrap_01_l\"),Transform=(Rotation=(X=-0.0804,Y=0.3175,Z=-0.1023,W=0.9393),Translation=(X=0,Y=0,Z=0),Scale3D=(X=1,Y=1,Z=1))),(Item=(Type=Bone,Name=\"dyn_StomachStrap_01_r\"),Transform=(Rotation=(X=0.1133,Y=0.3251,Z=0.0445,W=0 ...(truncated)"
    WeightMinimum=0
    Weight=@n22.OutputParam
   )
   > &RigUnit_ModifyTransforms(
    ItemToModify=((Item=(Type=Bone,Name="dyn_StomachStrap_01_l"),Transform=(Rotation=(X=-0.0122,Y=0.342,Z=0.2064,W=0.9167),Translation=(X=0,Y=0,Z=0),Scale3D=(X=1,Y=1,Z=1))))
    WeightMinimum=0
    Weight=@n18.OutputParam
   )
   > /Script/ControlRig.RigUnit_PositionConstraintLocalSpaceOffset(
    Child=(Type=Bone,Name="dyn_StomachStrap_base_l")
    Filter=(bX=True,bY=True,bZ=True)
    Parents=((Item=(Type=Bone,Name="spine_01"),Weight=0.8),(Item=(Type=Bone,Name="spine_05"),Weight=0.2))
    bMaintainOffset=True
    Weight=@n18.OutputParam
   ) {is=struct RigVMUnitNode}
   > &RigUnit_ModifyTransforms(
    ItemToModify=((Item=(Type=Bone,Name="dyn_StomachStrap_01_r"),Transform=(Rotation=(X=0.0414,Y=0.2276,Z=-0.1739,W=0.9572),Translation=(X=0,Y=0,Z=0),Scale3D=(X=1,Y=1,Z=1))))
    WeightMinimum=0
    Weight=@n19.OutputParam
   )
   Remap = &RigVMFunction_MathDoubleRemap(
    SourceMaximum=45
    SourceMinimum=55
    TargetMaximum=1
    TargetMinimum=0
    Value=&RigUnit_GetRelativeTransformForItem(Child=(Type=Bone,Name="pelvis"), Parent=(Type=Bone,Name="root"), bChildInitial=False, bParentInitial=False).RelativeTransform.Translation.Z
   )
   &RigUnit_ModifyTransforms(
    ItemToModify=((Item=(Type=Bone,Name="deform_glute_l"),Transform=(Rotation=(X=0,Y=0,Z=0,W=-1),Translation=(X=-2,Y=0,Z=0),Scale3D=(X=1,Y=1,Z=1))))
    WeightMinimum=0
    Weight=@Remap.Result
    ExecutePin=@RigVMFunction_Sequence.F
   )
   RigVMFunction_ControlFlowBranch = &RigVMFunction_ControlFlowBranch(ExecuteContext=(), Condition=RigVMVariableNode() {variable=is Playing Emote}.Value)
  ]}
]
hierarchy [
 root {loc=<0 0 0>, rot=<0 0 0 1>}
 pelvis {parent=root, loc=<0 -0.5301 98.0571>, rot=<0.0127 0.707 -0.0127 -0.707>}
 spine_01 {parent=pelvis, loc=<6.1208 -0.9328 0>, rot=<0 0 0.034 -0.9994>}
 spine_02 {parent=spine_01, loc=<6.5996 -0.2095 0>, rot=<0 0 0.0238 -0.9997>}
 spine_03 {parent=spine_02, loc=<6.4914 -0.1237 0>, rot=<0 0 0.0149 -0.9999>}
 spine_04 {parent=spine_03, loc=<6.9598 0.0794 0>, rot=<0 0 -0.0132 -0.9999>}
 spine_05 {parent=spine_04, loc=<6.9987 0 0>, rot=<0 0 -0.0015 -1>}
 neck_01 {parent=spine_05, loc=<11.3138 0.1529 0>, rot=<0 0 -0.2003 -0.9797>}
 neck_02 {parent=neck_01, loc=<4.5627 0 0>, rot=<0 0 0.0504 -0.9987>}
 head {parent=neck_02, loc=<5.3914 0 0>, rot=<0 0 0.115 -0.9934>}
 helmet_base {parent=head, loc=<0 0 0>, rot=<0 0 0 1>}
 dyn_helmet_ear_base_l {parent=helmet_base, loc=<10.3518 1.737 -6.9265>, rot=<0.4374 -0.2932 0.2006 -0.8261>}
 dyn_helmet_ear_01_l {parent=dyn_helmet_ear_base_l, loc=<0 0 0>, rot=<0 0 0 1>}
 dyn_helmet_ear_base_r {parent=helmet_base, loc=<10.3518 1.737 6.9265>, rot=<0.8261 -0.2006 0.2932 -0.4374>}
 dyn_helmet_ear_01_r {parent=dyn_helmet_ear_base_r, loc=<0 0 0>, rot=<0 0 0 1>}
 dyn_helmet_hair_base_a {parent=helmet_base, loc=<-1.027 1.7942 -7.2806>, rot=<0.2517 0.6466 0.6585 0.2912>}
 dyn_helmet_hair_01_a {parent=dyn_helmet_hair_base_a, loc=<0 0 0>, rot=<0 0 0 1>}
 dyn_helmet_hair_base_b {parent=helmet_base, loc=<-4.2378 -1.4023 -7.4483>, rot=<0.4777 0.5184 0.6923 0.1542>}
 dyn_helmet_hair_01_b {parent=dyn_helmet_hair_base_b, loc=<0 0 0>, rot=<0 0 0 1>}
 dyn_helmet_hair_base_с {parent=helmet_base, loc=<-3.5727 -6.2227 -3.0098>, rot=<0.0878 -0.9334 -0.3463 -0.0352>}
 dyn_helmet_hair_01_с {parent=dyn_helmet_hair_base_с, loc=<0 0 0>, rot=<0 0 0 1>}
 dyn_helmet_hair_base_d {parent=helmet_base, loc=<-4.0852 -5.8066 2.729>, rot=<0.0823 -0.9256 0.3649 0.0577>}
 dyn_helmet_hair_01_d {parent=dyn_helmet_hair_base_d, loc=<0 0 0>, rot=<0 0 0 1>}
 dyn_helmet_hair_base_e {parent=helmet_base, loc=<-2.6142 -2.8598 6.2365>, rot=<-0.1171 -0.8242 0.5477 0.0837>}
 dyn_helmet_hair_01_e {parent=dyn_helmet_hair_base_e, loc=<0 0 0>, rot=<0 0 0 1>}
 dyn_helmet_hair_base_f {parent=helmet_base, loc=<-1.5028 1.6445 6.7171>, rot=<-0.3401 -0.5725 0.6961 0.2683>}
 dyn_helmet_hair_01_f {parent=dyn_helmet_hair_base_f, loc=<0 0 0>, rot=<0 0 0 1>}
 hair_base {parent=head, loc=<0 0 0>, rot=<0 0 0 1>}
 dyn_hair_hair_base_b {parent=hair_base, loc=<6.4448 4.7169 -6.8953>, rot=<0.2528 -0.9223 -0.2315 -0.1783>}
 dyn_hair_hair_01_b {parent=dyn_hair_hair_base_b, loc=<0 0 0>, rot=<0 0 0 1>}
 dyn_hair_hair_base_a {parent=hair_base, loc=<7.5913 7.1636 -4.424>, rot=<-0.132 -0.9104 -0.3868 -0.0642>}
 dyn_hair_hair_01_a {parent=dyn_hair_hair_base_a, loc=<0 0 0>, rot=<0 0 0 1>}
 dyn_hair_hair_base_c {parent=hair_base, loc=<-0.9711 2.3495 -7.4924>, rot=<-0.3615 -0.8993 -0.2398 0.055>}
 dyn_hair_hair_01_c {parent=dyn_hair_hair_base_c, loc=<0 0 0>, rot=<0 0 0 1>}
 dyn_hair_hair_base_d {parent=hair_base, loc=<-3.8707 -1.1317 -7.3175>, rot=<-0.3395 -0.9303 -0.0288 0.1359>}
 dyn_hair_hair_01_d {parent=dyn_hair_hair_base_d, loc=<0 0 0>, rot=<0 0 0 1>}
 dyn_hair_hair_base_e {parent=hair_base, loc=<-2.282 -6.7869 -0.3971>, rot=<0.0537 -0.721 0.6849 -0.0908>}
 dyn_hair_hair_01_e {parent=dyn_hair_hair_base_e, loc=<0 0 0>, rot=<0 0 0 1>}
 dyn_hair_hair_base_f {parent=hair_base, loc=<-2.6893 -3.2301 5.5005>, rot=<-0.0092 -0.3658 0.9206 0.1363>}
 dyn_hair_hair_01_f {parent=dyn_hair_hair_base_f, loc=<0 0 0>, rot=<0 0 0 1>}
 dyn_hair_hair_base_g {parent=hair_base, loc=<-0.0534 0.709 8.1129>, rot=<0.0273 0.0524 0.9337 0.3531>}
 dyn_hair_hair_01_g {parent=dyn_hair_hair_base_g, loc=<0 0 0>, rot=<0 0 0 1>}
 dyn_hair_hair_base_h {parent=hair_base, loc=<6.3644 6.5128 6.1265>, rot=<0.2662 0.2294 0.9238 -0.1518>}
 dyn_hair_hair_01_h {parent=dyn_hair_hair_base_h, loc=<0 0 0>, rot=<0 0 0 1>}
 dyn_hair_hair_base_i {parent=hair_base, loc=<6.9726 9.6019 0.7525>, rot=<0.1903 0.6045 0.7695 -0.0782>}
 dyn_hair_hair_01_i {parent=dyn_hair_hair_base_i, loc=<0 0 0>, rot=<0 0 0 1>}
 dyn_hair_hair_base_j {parent=hair_base, loc=<8.3732 8.3868 -2.5291>, rot=<-0.2226 -0.7652 -0.6015 -0.056>}
 dyn_hair_hair_01_j {parent=dyn_hair_hair_base_j, loc=<0 0 0>, rot=<0 0 0 1>}
 clavicle_l {parent=spine_05, loc=<5.4831 -0.1053 -3.2946>, rot=<0.0465 -0.7289 -0.009 -0.6829>}
 upperarm_l {parent=clavicle_l, loc=<9.8658 -1.2136 -0.4105>, rot=<0.0164 -0.3245 -0.0269 -0.9454>}
 lowerarm_l {parent=upperarm_l, loc=<26.465 -0.8345 0.259>, rot=<0.0036 0.0161 -0.2298 -0.9731>}
 hand_l {parent=lowerarm_l, loc=<22.7355 -0.3203 -0.6105>, rot=<-0.6784 -0.0223 0.0735 -0.7307>}
 thumb_01_l {parent=hand_l, loc=<1.3118 -0.3191 -2.459>, rot=<0.6546 -0.4042 0.0664 -0.6354>}
 thumb_02_l {parent=thumb_01_l, loc=<5.9336 -0.0448 0.0401>, rot=<-0.0037 0.0213 0.298 -0.9543>}
 thumb_03_l {parent=thumb_02_l, loc=<2.2223 0.3682 -0.0907>, rot=<0.0118 -0.013 0.0066 -0.9998>}
 index_metacarpal_l {parent=hand_l, loc=<1.6535 0.0183 -2.96>, rot=<0.0158 -0.0124 -0.0173 -0.9996>}
 index_01_l {parent=index_metacarpal_l, loc=<8.3008 0.0357 -0.0524>, rot=<-0.0044 0.0485 0.0924 -0.9945>}
 index_02_l {parent=index_01_l, loc=<3.3587 -0.3004 0.0452>, rot=<-0.002 0.0027 0.3225 -0.9466>}
 index_03_l {parent=index_02_l, loc=<2.5578 -0.123 -0.0252>, rot=<-0.002 0.0067 0.2558 -0.9667>}
 middle_metacarpal_l {parent=hand_l, loc=<2.1856 -0.1725 -0.5984>, rot=<0.006 0.0168 -0.0329 -0.9993>}
 middle_01_l {parent=middle_metacarpal_l, loc=<7.3275 -0.2447 0.0129>, rot=<-0.0208 0.0042 0.0858 -0.9961>}
 middle_02_l {parent=middle_01_l, loc=<4.0123 0.1401 0.1175>, rot=<-0.0218 0.0258 0.3512 -0.9357>}
 middle_03_l {parent=middle_02_l, loc=<2.6077 0.0534 0.1373>, rot=<0.0082 0.0188 0.1708 -0.9851>}
 ring_metacarpal_l {parent=hand_l, loc=<2.2074 -0.3625 1.3638>, rot=<-0.0284 0.0722 -0.0458 -0.9959>}
 ring_01_l {parent=ring_metacarpal_l, loc=<7.0295 -0.6197 -0.1446>, rot=<-0.0489 -0.0252 0.1892 -0.9804>}
 ring_02_l {parent=ring_01_l, loc=<3.6174 0.4647 -0.1027>, rot=<0 0 0.2806 -0.9598>}
 ring_03_l {parent=ring_02_l, loc=<2.4152 0.114 0.0199>, rot=<0 0 0.1189 -0.9929>}
 pinky_metacarpal_l {parent=hand_l, loc=<1.7616 -0.9753 3.3814>, rot=<-0.0397 0.0814 -0.069 -0.9935>}
 pinky_01_l {parent=pinky_metacarpal_l, loc=<7.1709 -0.4831 -0.1094>, rot=<-0.0654 -0.0035 0.2242 -0.9723>}
 pinky_02_l {parent=pinky_01_l, loc=<2.8274 0.0893 -0.1043>, rot=<0.0037 0.0011 0.2854 -0.9584>}
 pinky_03_l {parent=pinky_02_l, loc=<2.2046 0.0307 -0.0078>, rot=<-0.0002 0.0098 0.1445 -0.9895>}
 deform_wrist_dn_base_l {parent=hand_l, loc=<0 -1.5326 0>, rot=<0 0 0 -1>}
 deform_wrist_dn_l {parent=deform_wrist_dn_base_l, loc=<0 0 0>, rot=<0 0 0 1>}
 deform_wrist_up_base_l {parent=hand_l, loc=<0 2.3009 0>, rot=<0 0 0 -1>}
 deform_wrist_up_l {parent=deform_wrist_up_base_l, loc=<0 0 0>, rot=<0 0 0 1>}
 hand_attach_l {parent=hand_l, loc=<8.5 -3 0>, rot=<0.7071 0.7071 0 0>}
 weapon_l {parent=hand_l, loc=<-0.1974 -3.4111 0.3811>, rot=<0 0 0 1>}
 lowerarm_twist_02_l {parent=lowerarm_l, loc=<10.6707 0 0>, rot=<0 0 0 1>}
 lowerarm_twist_01_l {parent=lowerarm_l, loc=<18.316 0 0>, rot=<0 0 0 1>}
 deform_elbow_in_base_l {parent=lowerarm_l, loc=<0.1834 3.6062 -0.2572>, rot=<0 0 0 -1>}
 deform_elbow_in_l {parent=deform_elbow_in_base_l, loc=<0 0 0>, rot=<0 0 0 1>}
 deform_elbow_out_base_l {parent=lowerarm_l, loc=<0 -4 0>, rot=<0 0 0 1>}
 deform_elbow_out_l {parent=deform_elbow_out_base_l, loc=<0 0 0>, rot=<0 0 0 1>}
 upperarm_twist_01_l {parent=upperarm_l, loc=<7.2135 -0.1134 0>, rot=<0 0 0 1>}
 deform_upperarm_bk_base_l {parent=upperarm_twist_01_l, loc=<-6.4053 -4.3746 0.314>, rot=<0 0 0 -1>}
 deform_upperarm_bk_l {parent=deform_upperarm_bk_base_l, loc=<0 0 0>, rot=<0 0 0 1>}
 deform_upperarm_fr_base_l {parent=upperarm_twist_01_l, loc=<-5.6312 5.092 0.3354>, rot=<0 0 0 -1>}
 deform_upperarm_fr_l {parent=deform_upperarm_fr_base_l, loc=<0 0 0>, rot=<0 0 0 1>}
 deform_upperarm_up_base_l {parent=upperarm_twist_01_l, loc=<-8.891 0.4825 3.9517>, rot=<0 0 0 -1>}
 deform_upperarm_up_l {parent=deform_upperarm_up_base_l, loc=<0 0 0>, rot=<0 0 0 1>}
 upperarm_twist_02_l {parent=upperarm_l, loc=<19.7998 -0.1134 0>, rot=<0 0 0 1>}
 clavicle_r {parent=spine_05, loc=<5.4831 -0.1053 3.2946>, rot=<0.7289 0.0465 -0.6829 0.009>}
 upperarm_r {parent=clavicle_r, loc=<-9.8658 1.2136 0.4104>, rot=<0.0164 -0.3245 -0.0269 -0.9454>}
 lowerarm_r {parent=upperarm_r, loc=<-26.4651 0.8345 -0.259>, rot=<0.0036 0.0161 -0.2298 -0.9731>}
 hand_r {parent=lowerarm_r, loc=<-22.7356 0.3203 0.6106>, rot=<-0.6784 -0.0223 0.0735 -0.7307>}
 thumb_01_r {parent=hand_r, loc=<-1.3119 0.3192 2.459>, rot=<0.6546 -0.4042 0.0664 -0.6354>}
 thumb_02_r {parent=thumb_01_r, loc=<-5.9335 0.0447 -0.0401>, rot=<-0.0037 0.0213 0.298 -0.9543>}
 thumb_03_r {parent=thumb_02_r, loc=<-2.2223 -0.3682 0.0907>, rot=<0.0118 -0.013 0.0066 -0.9998>}
 index_metacarpal_r {parent=hand_r, loc=<-1.6534 -0.0181 2.9601>, rot=<0.0158 -0.0124 -0.0173 -0.9996>}
 index_01_r {parent=index_metacarpal_r, loc=<-8.3008 -0.0357 0.0524>, rot=<-0.0044 0.0485 0.0924 -0.9945>}
 index_02_r {parent=index_01_r, loc=<-3.3589 0.3003 -0.0453>, rot=<-0.002 0.0027 0.3225 -0.9466>}
 index_03_r {parent=index_02_r, loc=<-2.5578 0.123 0.0252>, rot=<-0.002 0.0067 0.2558 -0.9667>}
 middle_metacarpal_r {parent=hand_r, loc=<-2.1856 0.1726 0.5984>, rot=<0.006 0.0168 -0.0329 -0.9993>}
 middle_01_r {parent=middle_metacarpal_r, loc=<-7.3273 0.2447 -0.0129>, rot=<-0.0208 0.0042 0.0858 -0.9961>}
 middle_02_r {parent=middle_01_r, loc=<-4.0124 -0.14 -0.1175>, rot=<-0.0218 0.0258 0.3512 -0.9357>}
 middle_03_r {parent=middle_02_r, loc=<-2.6077 -0.0533 -0.1373>, rot=<0.0082 0.0188 0.1708 -0.9851>}
 ring_metacarpal_r {parent=hand_r, loc=<-2.2073 0.3627 -1.3638>, rot=<-0.0284 0.0722 -0.0458 -0.9959>}
 ring_01_r {parent=ring_metacarpal_r, loc=<-7.0297 0.6197 0.1446>, rot=<-0.0489 -0.0252 0.1892 -0.9804>}
 ring_02_r {parent=ring_01_r, loc=<-3.6171 -0.4648 0.1027>, rot=<0 0 0.2806 -0.9598>}
 ring_03_r {parent=ring_02_r, loc=<-2.4153 -0.1141 -0.0199>, rot=<0 0 0.1189 -0.9929>}
 pinky_metacarpal_r {parent=hand_r, loc=<-1.7617 0.9753 -3.3814>, rot=<-0.0397 0.0814 -0.069 -0.9935>}
 pinky_01_r {parent=pinky_metacarpal_r, loc=<-7.1707 0.4832 0.1094>, rot=<-0.0654 -0.0035 0.2242 -0.9723>}
 pinky_02_r {parent=pinky_01_r, loc=<-2.8274 -0.0892 0.1043>, rot=<0.0037 0.0011 0.2854 -0.9584>}
 pinky_03_r {parent=pinky_02_r, loc=<-2.2045 -0.031 0.0078>, rot=<-0.0002 0.0098 0.1445 -0.9895>}
 deform_wrist_dn_base_r {parent=hand_r, loc=<-0.0001 1.5328 0>, rot=<0 0 0 -1>}
 deform_wrist_dn_r {parent=deform_wrist_dn_base_r, loc=<0 0 0>, rot=<0 0 0 1>}
 deform_wrist_up_base_r {parent=hand_r, loc=<-0.0002 -2.3007 -0.0001>, rot=<0 0 0 -1>}
 deform_wrist_up_r {parent=deform_wrist_up_base_r, loc=<0 0 0>, rot=<0 0 0 1>}
 weapon_r {parent=hand_r, loc=<0.1975 3.4112 -0.3811>, rot=<0 0 0 -1>}
 hand_attach_r {parent=hand_r, loc=<-8.5001 2.9999 0>, rot=<0 0 -0.7071 -0.7071>}
 lowerarm_twist_02_r {parent=lowerarm_r, loc=<-10.6707 0 0>, rot=<0 0 0 1>}
 lowerarm_twist_01_r {parent=lowerarm_r, loc=<-18.3161 0 0>, rot=<0 0 0 1>}
 dyn_charm_master {parent=lowerarm_twist_01_r, loc=<-0.6634 3.1259 0.3686>, rot=<-0.406 -0.6368 -0.5931 -0.279>}
 dyn_charm_01 {parent=dyn_charm_master, loc=<0 0 0>, rot=<0 0 0 1>}
 deform_elbow_in_base_r {parent=lowerarm_r, loc=<-0.1834 -3.6062 0.2572>, rot=<0 0 0 -1>}
 deform_elbow_in_r {parent=deform_elbow_in_base_r, loc=<0 0 0>, rot=<0 0 0 1>}
 deform_elbow_out_base_r {parent=lowerarm_r, loc=<0 4 0>, rot=<0 0 0 -1>}
 deform_elbow_out_r {parent=deform_elbow_out_base_r, loc=<0 0 0>, rot=<0 0 0 1>}
 upperarm_twist_01_r {parent=upperarm_r, loc=<-7.2135 0.1134 0>, rot=<0 0 0 1>}
 deform_upperarm_bk_base_r {parent=upperarm_twist_01_r, loc=<6.4052 4.3746 -0.3139>, rot=<0 0 0 -1>}
 deform_upperarm_bk_r {parent=deform_upperarm_bk_base_r, loc=<0 0 0>, rot=<0 0 0 1>}
 deform_upperarm_fr_base_r {parent=upperarm_twist_01_r, loc=<5.6311 -5.092 -0.3353>, rot=<0 0 0 -1>}
 deform_upperarm_fr_r {parent=deform_upperarm_fr_base_r, loc=<0 0 0>, rot=<0 0 0 1>}
 deform_upperarm_up_base_r {parent=upperarm_twist_01_r, loc=<8.8909 -0.4825 -3.9516>, rot=<0 0 0 -1>}
 deform_upperarm_up_r {parent=deform_upperarm_up_base_r, loc=<0 0 0>, rot=<0 0 0 1>}
 upperarm_twist_02_r {parent=upperarm_r, loc=<-19.7998 0.1134 0>, rot=<0 0 0 1>}
 attach_backpack {parent=spine_05, loc=<-0.123 -13.473 0>, rot=<-0.0093 -0.707 0.0093 -0.707>}
 Backpack_BR {parent=attach_backpack, type=Null, loc=<0 0 0>, rot=<0 0 0 1>}
 attach_cape {parent=spine_05, loc=<9.017 -11.714 0>, rot=<0.0616 -0.7044 -0.0616 -0.7044>}
 cape {parent=attach_cape, type=Null, loc=<0 0 0>, rot=<0 0 0 1>}
 deform_pec_aim_base_l {parent=spine_04, loc=<3.2456 9.3309 -1.2237>, rot=<0.0603 -0.4033 0.3318 -0.8507>}
 deform_pec_aim_l {parent=deform_pec_aim_base_l, loc=<0 0 0>, rot=<0 0 0 1>}
 deform_pec_base_l {parent=deform_pec_aim_l, loc=<9.3665 0 0>, rot=<0 0 0 1>}
 deform_pec_l {parent=deform_pec_base_l, loc=<0 0 0>, rot=<0 0 0 1>}
 deform_pec_aim_base_r {parent=spine_04, loc=<3.2452 9.3309 1.2237>, rot=<-0.4033 -0.0603 0.8507 0.3318>}
 deform_pec_aim_r {parent=deform_pec_aim_base_r, loc=<0 0 0>, rot=<0 0 0 -1>}
 deform_pec_base_r {parent=deform_pec_aim_r, loc=<-9.3667 -0.0003 -0.0004>, rot=<0 0 0 -1>}
 deform_pec_r {parent=deform_pec_base_r, loc=<0 0 0>, rot=<0 0 0 -1>}
 Holster_l_base {parent=spine_04, loc=<0.4889 -0.451 -12.6724>, rot=<0.0294 -0.7065 -0.0294 -0.7065>}
 Holster_L {parent=Holster_l_base, loc=<0 0 0>, rot=<0 0 0 1>}
 Holster_r_base {parent=spine_04, loc=<0.4887 -0.451 12.6724>, rot=<0.7065 0.0294 -0.7065 0.0294>}
 Holster_r {parent=Holster_r_base, loc=<0 0 0>, rot=<0 0 0 1>}
 dyn_knife_base {parent=spine_01, loc=<0.4803 -1.4781 13.7241>, rot=<0.1914 -0.1007 0.9572 -0.1922>}
 dyn_knife_01 {parent=dyn_knife_base, loc=<2.1602 0 0>, rot=<0 0 0 1>}
 dyn_knife_02 {parent=dyn_knife_01, loc=<3.6028 0 0>, rot=<0 0 0 1>}
 dyn_knife_03 {parent=dyn_knife_02, loc=<0.7193 0.6474 1.6049>, rot=<0.0672 -0.1498 -0.1563 -0.974>}
 dyn_strapFront_base_r {parent=spine_01, loc=<-4.4804 5.5758 12.2403>, rot=<-0.0223 -0.8995 -0.3367 0.2774>}
 dyn_strapFront_01_r {parent=dyn_strapFront_base_r, loc=<0 0 0>, rot=<0 0 0 1>}
 dyn_strapFront_02_r {parent=dyn_strapFront_01_r, loc=<7.325 0 0>, rot=<0 0 0 1>}
 dyn_strapBack_base_r {parent=spine_01, loc=<-1.5821 -9.2746 6.7785>, rot=<0.2622 -0.4336 0.8618 0.0256>}
 dyn_strapBack_01_r {parent=dyn_strapBack_base_r, loc=<0 0 0>, rot=<0 0 0 1>}
 dyn_strapBack_02_r {parent=dyn_strapBack_01_r, loc=<7.1127 0 0>, rot=<0 0 0 1>}
 dyn_strapBack_03_r {parent=dyn_strapBack_02_r, loc=<7.071 0 0>, rot=<0.0275 -0.1549 -0.1044 -0.982>}
 dyn_strapBack_04_r {parent=dyn_strapBack_03_r, loc=<6.6674 0 0>, rot=<-0.6079 -0.3686 -0.138 -0.6896>}
 dyn_strapBack_05_r {parent=dyn_strapBack_04_r, loc=<5.0517 0 0>, rot=<-0.0755 -0.4064 0.2781 -0.8671>}
 dyn_strapBack_06_r {parent=dyn_strapBack_05_r, loc=<5.7313 0 0>, rot=<-0.6722 0.0518 0.2049 -0.7096>}
 dyn_strapFront_base_l {parent=spine_01, loc=<-4.3382 5.4434 -13.5243>, rot=<0.1894 -0.8995 -0.3208 -0.2282>}
 dyn_strapFront_01_l {parent=dyn_strapFront_base_l, loc=<0 0 0>, rot=<0 0 0 1>}
 dyn_strapFront_02_l {parent=dyn_strapFront_01_l, loc=<5.2904 0 0>, rot=<-0.2963 -0.1324 0.0231 -0.9456>}
 dyn_strapBack_base_l {parent=spine_01, loc=<-1.4506 -8.8784 -7.2689>, rot=<0.1573 -0.9077 -0.3233 -0.2164>}
 dyn_strapBack_01_l {parent=dyn_strapBack_base_l, loc=<0 0 0>, rot=<0 0 0 1>}
 dyn_strapBack_02_l {parent=dyn_strapBack_01_l, loc=<6.8339 0 0>, rot=<0 0 0 1>}
 dyn_strapBack_03_l {parent=dyn_strapBack_02_l, loc=<7.5793 0 0>, rot=<-0.0836 0.1987 -0.1648 -0.9625>}
 dyn_strapBack_04_l {parent=dyn_strapBack_03_l, loc=<5.6366 0 0>, rot=<0.6637 0.3983 -0.1458 -0.6161>}
 dyn_strapBack_05_l {parent=dyn_strapBack_04_l, loc=<6.0064 0 0>, rot=<-0.591 0 0 -0.8067>}
 dyn_strapBack_06_l {parent=dyn_strapBack_05_l, loc=<7.0185 0 0>, rot=<0 0 0 1>}
 dyn_StomachStrap_base_l {parent=spine_01, loc=<-1.5547 8.5077 -6.3508>, rot=<-0.576 -0.0179 -0.0017 -0.8172>}
 dyn_StomachStrap_01_l {parent=dyn_StomachStrap_base_l, loc=<0 0 0>, rot=<0 0 0 1>}
 dyn_StomachStrap_base_r {parent=spine_01, loc=<-1.5546 8.5077 6.3508>, rot=<-0.8206 -0.0174 0.0045 -0.5713>}
 dyn_StomachStrap_01_r {parent=dyn_StomachStrap_base_r, loc=<0 0 0>, rot=<0 0 0 1>}
 dyn_beltCorr_base_r {parent=spine_01, loc=<-3.4689 5.8623 8.482>, rot=<0.6256 0.4503 0.5584 0.3065>}
 dyn_beltCorr_r {parent=dyn_beltCorr_base_r, loc=<0 0 0>, rot=<0 0 0 1>}
 dyn_beltCorr_base_l {parent=spine_01, loc=<-3.4689 5.8623 -8.4815>, rot=<-0.3065 -0.5584 -0.4503 -0.6256>}
 dyn_beltCorr_l {parent=dyn_beltCorr_base_l, loc=<0 0 0>, rot=<0 0 0 1>}
 thigh_l {parent=pelvis, loc=<-1.0855 -0.2109 -8.7053>, rot=<0.0009 0.0355 0.0155 -0.9993>}
 calf_l {parent=thigh_l, loc=<-42.5284 -0.0019 -0.7297>, rot=<0.0006 -0.0161 -0.0391 -0.9991>}
 foot_l {parent=calf_l, loc=<-46.7402 -2.5014 -0.592>, rot=<0.0427 -0.0238 0.0353 -0.9982>}
 ball_l {parent=foot_l, loc=<-3.5658 10.8066 -0.1544>, rot=<0.0001 -0.0001 -0.7063 -0.7079>}
 def_foot_fr_base_l {parent=foot_l, loc=<4.2983 7.7407 0.2033>, rot=<-0.598 -0.3925 -0.4332 -0.5484>}
 def_foot_fr_01_l {parent=def_foot_fr_base_l, loc=<0 0 0>, rot=<0 0 0 1>}
 def_foot_bk_base_l {parent=foot_l, loc=<4.2264 -3.6263 0.4368>, rot=<0.4729 -0.5234 0.4807 -0.5208>}
 def_foot_bk_01_l {parent=def_foot_bk_base_l, loc=<0 0 0>, rot=<0 0 0 1>}
 calf_twist_02_l {parent=calf_l, loc=<-19.1546 -1.7943 -0.2638>, rot=<0.0237 0.0095 -0.0035 -0.9997>}
 calf_twist_01_l {parent=calf_l, loc=<-35.8483 -1.9197 -0.5794>, rot=<0.0237 0.0095 -0.0035 -0.9997>}
 deform_knee_in_base_l {parent=calf_l, loc=<0.8031 -5.0231 0.2456>, rot=<0 0 0 -1>}
 deform_knee_in_l {parent=deform_knee_in_base_l, loc=<0 0 0>, rot=<0 0 0 1>}
 deform_knee_out_base_l {parent=calf_l, loc=<0 4 0>, rot=<0 0 0 -1>}
 deform_knee_out_l {parent=deform_knee_out_base_l, loc=<0 0 0>, rot=<0 0 0 1>}
 thigh_twist_01_l {parent=thigh_l, loc=<-27.952 0 0>, rot=<0.0013 0 0 -1>}
 dyn_grenade_base_a {parent=thigh_twist_01_l, loc=<6.4365 4.7678 -7.7891>, rot=<-0.1624 -0.7135 -0.1482 -0.6653>}
 dyn_grenade_a {parent=dyn_grenade_base_a, loc=<0 0 0>, rot=<0 0 0 1>}
 dyn_grenade_base_b {parent=thigh_twist_01_l, loc=<6.3942 -0.0312 -8.3039>, rot=<0.0689 -0.7285 0.0674 -0.6782>}
 dyn_grenade_b {parent=dyn_grenade_base_b, loc=<0 0 0>, rot=<0 0 0 1>}
 dyn_thigh_strap_01_l {parent=thigh_twist_01_l, loc=<27.952 0 0>, rot=<0.0407 -0.9319 0.0779 -0.3519>}
 dyn_thigh_strap_02_l {parent=dyn_thigh_strap_01_l, loc=<14.4302 0 0>, rot=<0 0 0 1>}
 deform_groin_base_l {parent=thigh_l, loc=<0.4801 5.6776 -0.2486>, rot=<0 0 0 -1>}
 deform_groin_l {parent=deform_groin_base_l, loc=<0 0 0>, rot=<0 0 0 1>}
 thigh_grenades_l {parent=thigh_l, loc=<-24.0721 0 0>, rot=<0.0013 0 0 -1>}
 thigh_r {parent=pelvis, loc=<-1.0853 -0.2109 8.7053>, rot=<0.0355 -0.0009 0.9993 0.0155>}
 calf_r {parent=thigh_r, loc=<42.5286 0.0019 0.7297>, rot=<0.0006 -0.0161 -0.0391 -0.9991>}
 foot_r {parent=calf_r, loc=<46.7403 2.5014 0.592>, rot=<0.0427 -0.0238 0.0353 -0.9982>}
 ball_r {parent=foot_r, loc=<3.5658 -10.8066 0.1544>, rot=<0.0001 -0.0001 -0.7063 -0.7079>}
 def_foot_bk_base_r {parent=foot_r, loc=<-4.2265 3.6263 -0.4368>, rot=<0.4807 0.5208 -0.4729 -0.5234>}
 def_foot_bk_01_r {parent=def_foot_bk_base_r, loc=<0 0 0>, rot=<0 0 0 1>}
 def_foot_fr_base_r {parent=foot_r, loc=<-4.2983 -7.7407 -0.2033>, rot=<-0.4332 0.5484 0.598 -0.3925>}
 def_foot_fr_01_r {parent=def_foot_fr_base_r, loc=<0 0 0>, rot=<0 0 0 1>}
 calf_twist_02_r {parent=calf_r, loc=<19.1546 1.7943 0.2638>, rot=<0.0237 0.0095 -0.0035 -0.9997>}
 calf_twist_01_r {parent=calf_r, loc=<35.8485 1.9197 0.5794>, rot=<0.0237 0.0095 -0.0035 -0.9997>}
 deform_knee_in_base_r {parent=calf_r, loc=<-0.8031 5.0231 -0.2457>, rot=<0 0 0 -1>}
 deform_knee_in_r {parent=deform_knee_in_base_r, loc=<0 0 0>, rot=<0 0 0 1>}
 deform_knee_out_base_r {parent=calf_r, loc=<-0.0001 -4 0>, rot=<0 0 0 -1>}
 deform_knee_out_r {parent=deform_knee_out_base_r, loc=<0.0001 0 0>, rot=<0 0 0 1>}
 def_kneePad_r {parent=deform_knee_out_r, loc=<1.6454 -2.2791 0>, rot=<0 0 0 1>}
 thigh_twist_01_r {parent=thigh_r, loc=<27.952 0 0>, rot=<0.0013 0 0 -1>}
 dyn_thigh_strap_01_r {parent=thigh_twist_01_r, loc=<-27.952 0 0>, rot=<0.9243 -0.0008 0.3816 -0.0018>}
 dyn_thigh_strap_02_r {parent=dyn_thigh_strap_01_r, loc=<14.4302 0 0>, rot=<0 0 0 1>}
 deform_groin_base_r {parent=thigh_r, loc=<-0.48 -5.6776 0.2486>, rot=<0 0 0 -1>}
 deform_groin_r {parent=deform_groin_base_r, loc=<0.0001 0 0>, rot=<0 0 0 1>}
 deform_glute_base_l {parent=pelvis, loc=<-7.3466 -9.0213 -9.1245>, rot=<0.0172 0.031 -0.4477 -0.8935>}
 deform_glute_l {parent=deform_glute_base_l, loc=<0 0 0>, rot=<0 0 0 1>}
 deform_glute_base_r {parent=pelvis, loc=<-7.3464 -9.0213 9.1245>, rot=<0.031 -0.0172 0.8935 -0.4477>}
 deform_glute_r {parent=deform_glute_base_r, loc=<0 0 0>, rot=<0 0 0 1>}
 simspace {parent=pelvis, loc=<0 0 0>, rot=<0 0 0 1>}
]
variables [
 is Playing Emote {type=bool, default=False}
 isAssaultRifle {type=bool, default=False}
 isFrontend {type=bool, default=False}
 isNotPlayingEmoteFloat {type=double, default=0}
 isPistol {type=bool, default=False}
 isSMG {type=bool, default=False}
 isTargeting {type=bool, default=False}
]
