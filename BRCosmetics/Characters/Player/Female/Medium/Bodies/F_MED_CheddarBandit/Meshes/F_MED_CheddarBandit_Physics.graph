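# Physics asset for the F_MED_CheddarBandit female player body, set up for secondary motion of a knife accessory.
# A four-bone chain runs from spine_01 through dyn_knife_base, 01, 02 and 03, with tight swing and twist limits that loosen slightly at the tip.
# Capsules cover the knife joints, with a box on the last bone.
# Unconstrained capsules on pelvis, spine_01, both thighs and right-side back and thigh straps provide body collision volumes.
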PhysicsAsset /BRCosmetics/Characters/Player/Female/Medium/Bodies/F_MED_CheddarBandit/Meshes/F_MED_CheddarBandit_Physics
bodies {
 dyn_knife_01 [{shapeType=Capsule, length=0.505, radius=0.505, shapeName=dyn_knife_01_capsule, rotation=<rot 0 -90 0>}]
 dyn_knife_02 [{
   shapeType=Capsule
   length=2.8147
   radius=0.5647
   shapeName=dyn_knife_02_capsule
   center=<2.0015 0.7841 -0.3469>
   rotation=<rot 0 -1.285 -74.7967>
  }]
 dyn_knife_03 [{
   shapeType=Box
   extentX=6.6549
   extentY=2.6896
   extentZ=33.4912
   shapeName=dyn_knife_03_box
   center=<1.172 0.0015 0.4548>
   rotation=<rot 5.8473 -89.7423 -88.6485>
  }]
 dyn_knife_base [{shapeType=Capsule, length=0.505, radius=0.505, shapeName=dyn_knife_base_capsule, rotation=<rot 0 -90 0>}]
 dyn_strapBack_02_r [{
   shapeType=Capsule
   length=34.7565
   radius=2.4979
   shapeName=dyn_strapBack_02_r_capsule
   center=<3.8643 -2.6879 -2.4402>
   rotation=<rot 0.0001 -135.9651 -70.1647>
  }]
 dyn_thigh_strap_02_r [{
   shapeType=Capsule
   length=37.0502
   radius=3.2404
   shapeName=dyn_thigh_strap_02_r_capsule
   center=<-8.1557 -6.6707 0.6616>
   rotation=<rot 3.409 -42.1764 -81.2782>
  }]
 pelvis [
  {
   shapeType=Capsule
   length=14.9983
   radius=8.4401
   shapeName=pelvis_capsule
   center=<-4.2356 -1.7199 0.0659>
   rotation=<rot 0.5134 -90 1.1751>
  }
  {shapeType=Capsule, length=7.375, radius=9.7852, shapeName=pelvis_capsule2, center=<-5.3286 -3.6083 2.1856>}
 ]
 spine_01 [{shapeType=Capsule, length=11.6648, radius=9.2155, shapeName=spine_01_capsule, center=<-1.238 0 0>, rotation=<rot 0 -90 0>}]
 thigh_l [{
   shapeType=Capsule
   length=14.202
   radius=8
   shapeName=thigh_l_capsule
   center=<-13.1319 0.4529 -0.3811>
   rotation=<rot -78.4476 -0.0004 -6.4993>
  }]
 thigh_r [{
   shapeType=Capsule
   length=13.6089
   radius=8.593
   shapeName=thigh_r_capsule
   center=<13.0788 -0.4529 0.6408>
   rotation=<rot -78.4476 -0.0004 173.5007>
  }]
}
constraints [
 {bone1Name=dyn_knife_base, bone2Name=spine_01, swing1LimitDegrees=5, swing2LimitDegrees=5, twistLimitDegrees=1}
 {bone1Name=dyn_knife_01, bone2Name=dyn_knife_base, swing1LimitDegrees=5, swing2LimitDegrees=5, twistLimitDegrees=1}
 {bone1Name=dyn_knife_02, bone2Name=dyn_knife_01, swing1LimitDegrees=5, swing2LimitDegrees=5, twistLimitDegrees=1}
 {bone1Name=dyn_knife_03, bone2Name=dyn_knife_02, swing1LimitDegrees=7, swing2LimitDegrees=7, twistLimitDegrees=3}
]
