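# Post-process style AnimBlueprint for the CheddarBandit female body, adding secondary motion to holsters, straps, charm, knife and grenades.
# Copies the parent mesh pose, applies weighted bone constraints and a Control Rig fed by emote, targeting, frontend and weapon flags, then strap and charm Anim Dynamics.
# That cached pose feeds RigidBody and knife-dynamics branches, joined by an unwired bool blend before grenade dynamics.
# The event graph updates weapon flags and gravity override.
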AnimBlueprint /BRCosmetics/Characters/Player/Female/Medium/Bodies/F_MED_CheddarBandit/Meshes/F_MED_CheddarBandit_AnimBP
target_skeleton=&/F_MED_CheddarBandit_Skeleton
graphs [
 AnimGraph {nodes [
   n1 = RigidBody!(ComponentPose=Local To Component(LocalPose=Use cached pose 'toSwitch'()), Alpha, OverrideWorldGravity:Vector="0,0,0", ExternalForce) -> (Pose) {
    bClampLinearTranslationLimitToRefPose=true
    bOverrideWorldGravity=true
    overridePhysicsAsset=&/F_MED_CheddarBandit_Physics
    simulationSpace=BaseBoneSpace
    alphaBoolBlend {blendOption=Linear}
    baseBoneRef {boneName=spine_01}
    componentAppliedLinearAccClamp=<3000 3000 3000>
    componentLinearAccScale=<0.5 0.5 0.5>
    componentLinearVelScale=<0.5 0.5 0.5>
    simSpaceSettings {
     dampingAlpha=0.1
     maxAngularAcceleration=100
     maxAngularVelocity=6
     maxLinearAcceleration=1000
     maxLinearVelocity=400
     worldAlpha=1
     externalLinearDragV=<0.5 0.5 0.5>
    }
   }
   n2 = Component To Local(ComponentPose=@n1)
   n3 = "Transform (Modify) Bone - Bone: dyn_knife_01"(
    ComponentPose=Local To Component(LocalPose=Use cached pose 'toSwitch'())
    Alpha=1
    Translation="0,0,0"
    Rotation="-12.156,0.000,-0.000"
    Scale="1,1,1"
   ) {rotationMode=BMM_Additive, alphaBoolBlend {blendOption=Linear}, boneToModify {boneName=dyn_knife_01}, rotation {pitch=-12.156}}
   n4 = "Transform (Modify) Bone - Bone: dyn_knife_02"(ComponentPose=@n3, Alpha=1, Translation="0,0,0", Rotation="-4.515,1.306,-6.161", Scale="1,1,1") {rotationMode=BMM_Additive, alphaBoolBlend {blendOption=Linear}, boneToModify {boneName=dyn_knife_02}, rotation=<rot -4.515 1.306 -6.161>}
   n5 = Anim Dynamics - Bone: dyn_knife_03!(ComponentPose=@n4, bAlphaBoolEnabled:bool=True, ExternalForce="0,0,0") -> (Pose) {
    alphaInputType=Bool
    angularDampingOverride=0.9999
    angularSpringConstant=2000
    bAngularSpring=true
    bOverrideAngularDamping=true
    bUseGravityOverride=true
    simulationSpace=BoneRelative
    alphaBoolBlend {blendOption=Linear}
    boundBone {boneName=dyn_knife_03}
    componentAppliedLinearAccClamp=<3000 3000 3000>
    componentLinearAccScale=<0.5 0.5 0.5>
    componentLinearVelScale=<0.5 0.5 0.5>
    gravityOverride {z=-980}
    physicsBodyDefinitions [{
      boundBone {boneName=dyn_knife_03}
      boxExtents=<18 18 18>
      constraintSetup {angularLimitsMax=<3 10 10>, angularLimitsMin=<-3 -10 -10>, angularTarget=<1 0 0>}
      localJointOffset=<0.5 0 0>
     }]
    relativeSpaceBone {boneName=pelvis}
   }
   n6 = Anim Dynamics - Bone: dyn_knife_03(ComponentPose=@n5, bAlphaBoolEnabled=True, ExternalForce="0,0,0") {
    alphaInputType=Bool
    angularDampingOverride=0.9999
    angularSpringConstant=500
    bAngularSpring=true
    bOverrideAngularDamping=true
    bUseGravityOverride=true
    simulationSpace=World
    alphaBoolBlend {blendOption=Linear}
    boundBone {boneName=dyn_knife_03}
    componentAppliedLinearAccClamp=<3000 3000 3000>
    componentLinearAccScale=<0.5 0.5 0.5>
    componentLinearVelScale=<0.5 0.5 0.5>
    gravityOverride {z=-980}
    physicsBodyDefinitions [{
      boundBone {boneName=dyn_knife_03}
      boxExtents=<20 20 20>
      constraintSetup {angularLimitsMax=<3 5 5>, angularLimitsMin=<-3 -5 -5>, angularTarget=<1 0 0>}
      localJointOffset=<0.1 0 0>
     }]
    relativeSpaceBone {boneName=pelvis}
   }
   n7 = Component To Local(ComponentPose=@n6)
   n8 = Anim Dynamics - Bone: dyn_grenade_a!(
    ComponentPose=Local To Component(LocalPose=Blend Poses by bool(BlendPose_0=@n2, BlendPose_1=@n7) {blendType=Linear})
    bAlphaBoolEnabled:bool=True
    ExternalForce="0,0,0"
   ) -> (Pose) {
    alphaInputType=Bool
    angularDampingOverride=0.999
    angularSpringConstant=1500
    bAngularSpring=true
    bOverrideAngularDamping=true
    bUseGravityOverride=true
    simulationSpace=BoneRelative
    alphaBoolBlend {blendOption=Linear}
    boundBone {boneName=dyn_grenade_a}
    componentAppliedLinearAccClamp=<3000 3000 3000>
    componentLinearAccScale=<0.7 0.7 0.7>
    componentLinearVelScale=<0.7 0.7 0.7>
    gravityOverride {z=-980}
    physicsBodyDefinitions [{
      boundBone {boneName=dyn_grenade_a}
      boxExtents=<10 10 10>
      constraintSetup {coneAngle=1, angularLimitsMax=<3 3 3>, angularLimitsMin=<-3 -3 -3>, angularTarget=<1 0 0>}
      localJointOffset=<1 0 0>
     }]
    relativeSpaceBone {boneName=thigh_r}
   }
   n9 = Anim Dynamics - Bone: dyn_grenade_b!(ComponentPose=@n8, bAlphaBoolEnabled:bool=True, ExternalForce="0,0,0") -> (Pose) {
    alphaInputType=Bool
    angularDampingOverride=0.999
    angularSpringConstant=1500
    bAngularSpring=true
    bOverrideAngularDamping=true
    bUseGravityOverride=true
    simulationSpace=BoneRelative
    alphaBoolBlend {blendOption=Linear}
    boundBone {boneName=dyn_grenade_b}
    componentAppliedLinearAccClamp=<3000 3000 3000>
    componentLinearAccScale=<0.7 0.7 0.7>
    componentLinearVelScale=<0.7 0.7 0.7>
    gravityOverride {z=-980}
    physicsBodyDefinitions [{
      boundBone {boneName=dyn_grenade_b}
      boxExtents=<10 10 10>
      constraintSetup {coneAngle=1, angularLimitsMax=<3 3 3>, angularLimitsMin=<-3 -3 -3>, angularTarget=<1 0 0>}
      localJointOffset=<1 0 0>
     }]
    relativeSpaceBone {boneName=thigh_r}
   }
   Output Pose(Result=Component To Local(ComponentPose=@n9)) {layerGroup=None}
   n10 = "Constraint - Holster_l_base (PPPPPP)"!(
    ComponentPose=Local To Component(LocalPose=Copy Pose From Mesh() {bUseAttachedParent=true})
    Alpha=1
    ConstraintWeights_0=0.1424
    ConstraintWeights_1:float=0.137
    ConstraintWeights_2:float=0.5079
    ConstraintWeights_3:float=0.1501
    ConstraintWeights_4:float=0.0325
    ConstraintWeights_5:float=0.03
   ) -> (Pose) {
    alphaBoolBlend {blendOption=Linear}
    boneToModify {boneName=Holster_l_base}
    constraintSetup [
     {targetBone {boneName=spine_05}}
     {targetBone {boneName=clavicle_l}}
     {targetBone {boneName=spine_04}}
     {targetBone {boneName=upperarm_twist_01_l}}
     {targetBone {boneName=spine_03}}
     {targetBone {boneName=spine_02}}
    ]
    constraintWeights [1,1,1,1,1,1]
   }
   n11 = "Constraint - Holster_r_base (PPPPPP)"!(
    ComponentPose=@n10
    Alpha=1
    ConstraintWeights_0=0.1424
    ConstraintWeights_1:float=0.137
    ConstraintWeights_2:float=0.5079
    ConstraintWeights_3:float=0.1501
    ConstraintWeights_4:float=0.0325
    ConstraintWeights_5:float=0.03
   ) -> (Pose) {
    alphaBoolBlend {blendOption=Linear}
    boneToModify {boneName=Holster_r_base}
    constraintSetup [
     {targetBone {boneName=spine_05}}
     {targetBone {boneName=clavicle_r}}
     {targetBone {boneName=spine_04}}
     {targetBone {boneName=upperarm_twist_01_r}}
     {targetBone {boneName=spine_03}}
     {targetBone {boneName=spine_02}}
    ]
    constraintWeights [1,1,1,1,1,1]
   }
   n12 = "Constraint - dyn_StomachStrap_base_l (PPPPPPPP)"!(
    ComponentPose=@n11
    Alpha=1
    ConstraintWeights_0=0.0031
    ConstraintWeights_1:float=0.0002
    ConstraintWeights_2:float=0.0592
    ConstraintWeights_3:float=0.5635
    ConstraintWeights_4:float=0.0061
    ConstraintWeights_5:float=0.023
    ConstraintWeights_6:float=0.3211
    ConstraintWeights_7:float=0.0239
   ) -> (Pose) {
    alphaBoolBlend {blendOption=Linear}
    boneToModify {boneName=dyn_StomachStrap_base_l}
    constraintSetup [
     {targetBone {boneName=spine_04}}
     {targetBone {boneName=spine_03}}
     {targetBone {boneName=spine_02}}
     {targetBone {boneName=pelvis}}
     {targetBone {boneName=thigh_twist_01_l}}
     {targetBone {boneName=deform_groin_l}}
     {targetBone {boneName=dyn_beltCorr_l}}
     {targetBone {boneName=dyn_beltCorr_r}}
    ]
    constraintWeights [1,1,1,1,1,1,1,1]
   }
   n13 = "Constraint - dyn_StomachStrap_base_r (PPPPPPPP)"!(
    ComponentPose=@n12
    Alpha=1
    ConstraintWeights_0=0.0031
    ConstraintWeights_1:float=0.0002
    ConstraintWeights_2:float=0.0592
    ConstraintWeights_3:float=0.5635
    ConstraintWeights_4:float=0.0061
    ConstraintWeights_5:float=0.023
    ConstraintWeights_6:float=0.0239
    ConstraintWeights_7:float=0.3211
   ) -> (Pose) {
    alphaBoolBlend {blendOption=Linear}
    boneToModify {boneName=dyn_StomachStrap_base_r}
    constraintSetup [
     {targetBone {boneName=spine_04}}
     {targetBone {boneName=spine_03}}
     {targetBone {boneName=spine_02}}
     {targetBone {boneName=pelvis}}
     {targetBone {boneName=thigh_twist_01_r}}
     {targetBone {boneName=deform_groin_r}}
     {targetBone {boneName=dyn_beltCorr_l}}
     {targetBone {boneName=dyn_beltCorr_r}}
    ]
    constraintWeights [1,1,1,1,1,1,1,1]
   }
   n14 = "Constraint - dyn_knife_base (PPPPP)"!(
    ComponentPose=@n13
    Alpha=1
    ConstraintWeights_0=0.0309
    ConstraintWeights_1:float=0.8941
    ConstraintWeights_2:float=0.0675
    ConstraintWeights_3:float=0.0065
    ConstraintWeights_4:float=0.001
   ) -> (Pose) {
    alphaBoolBlend {blendOption=Linear}
    boneToModify {boneName=dyn_knife_base}
    constraintSetup [
     {targetBone {boneName=spine_02}}
     {targetBone {boneName=pelvis}}
     {targetBone {boneName=thigh_twist_01_r}}
     {targetBone {boneName=deform_glute_r}}
     {targetBone {boneName=deform_groin_r}}
    ]
    constraintWeights [1,1,1,1,1]
   }
   n15 = Control Rig!(
    Source=Component To Local(ComponentPose=@n14)
    Alpha
    is Playing Emote:bool=Get bIsPlayingEmote()
    isTargeting:bool=Get bIsTargeting()
    isPistol:bool=Get isPistol() -> (isPistol:bool)
    isAssaultRifle:bool=Get isAssaultRifle() -> (isAssaultRifle:bool)
    isSMG:bool=Get isSMG() -> (isSMG:bool)
    isFrontend:bool=Get bIsInFrontEnd()
   ) -> (Pose) {
    alphaBoolBlend {blendOption=Linear}
    controlRigAssetReference {blueprintRigClass=&/F_MED_CheddarBandit_CtrlRig.F_MED_CheddarBandit_CtrlRig_C}
   }
   n16 = "Constraint - dyn_strapFront_base_l (PPPPPP)"!(
    ComponentPose=Local To Component(LocalPose=@n15)
    Alpha=1
    ConstraintWeights_0=0.0019
    ConstraintWeights_1:float=0.3619
    ConstraintWeights_2:float=0.1182
    ConstraintWeights_3:float=0.1054
    ConstraintWeights_4:float=0.4125
    ConstraintWeights_5:float=0.0001
   ) -> (Pose) {
    alphaBoolBlend {blendOption=Linear}
    boneToModify {boneName=dyn_strapFront_base_l}
    constraintSetup [
     {targetBone {boneName=spine_02}}
     {targetBone {boneName=pelvis}}
     {targetBone {boneName=thigh_twist_01_l}}
     {targetBone {boneName=deform_groin_l}}
     {targetBone {boneName=dyn_beltCorr_l}}
     {targetBone {boneName=dyn_beltCorr_r}}
    ]
    constraintWeights [1,1,1,1,1,1]
   }
   n17 = "Constraint - dyn_strapBack_base_l (PPPP)"!(
    ComponentPose=@n16
    Alpha=1
    ConstraintWeights_0=0.0392
    ConstraintWeights_1:float=0.0198
    ConstraintWeights_2:float=0.9003
    ConstraintWeights_3:float=0.0407
   ) -> (Pose) {
    alphaBoolBlend {blendOption=Linear}
    boneToModify {boneName=dyn_strapBack_base_l}
    constraintSetup [{targetBone {boneName=spine_02}}, {targetBone {boneName=spine_01}}, {targetBone {boneName=pelvis}}, {targetBone {boneName=deform_glute_l}}]
    constraintWeights [1,1,1,1]
   }
   n18 = "Constraint - dyn_strapFront_base_r (PPPPPP)"!(
    ComponentPose=@n17
    Alpha=1
    ConstraintWeights_0=0.0024
    ConstraintWeights_1:float=0.3699
    ConstraintWeights_2:float=0.1093
    ConstraintWeights_3:float=0.1146
    ConstraintWeights_4:float=0.0004
    ConstraintWeights_5:float=0.4034
   ) -> (Pose) {
    alphaBoolBlend {blendOption=Linear}
    boneToModify {boneName=dyn_strapFront_base_r}
    constraintSetup [
     {targetBone {boneName=spine_02}}
     {targetBone {boneName=pelvis}}
     {targetBone {boneName=thigh_twist_01_r}}
     {targetBone {boneName=deform_groin_r}}
     {targetBone {boneName=dyn_beltCorr_l}}
     {targetBone {boneName=dyn_beltCorr_r}}
    ]
    constraintWeights [1,1,1,1,1,1]
   }
   n19 = "Constraint - dyn_strapBack_base_r (PPPP)"!(
    ComponentPose=@n18
    Alpha=1
    ConstraintWeights_0=0.04
    ConstraintWeights_1:float=0.0205
    ConstraintWeights_2:float=0.9004
    ConstraintWeights_3:float=0.0391
   ) -> (Pose) {
    alphaBoolBlend {blendOption=Linear}
    boneToModify {boneName=dyn_strapBack_base_r}
    constraintSetup [{targetBone {boneName=spine_02}}, {targetBone {boneName=spine_01}}, {targetBone {boneName=pelvis}}, {targetBone {boneName=deform_glute_r}}]
    constraintWeights [1,1,1,1]
   }
   n20 = Anim Dynamics - Bone: dyn_charm_01(ComponentPose=Local To Component(LocalPose=Component To Local(ComponentPose=@n19)), Alpha=1, ExternalForce="0,0,0") {
    angularBiasOverride=0.5
    angularDampingOverride=0.99
    angularSpringConstant=50
    bAngularSpring=true
    bOverrideAngularDamping=true
    bUseGravityOverride=true
    simulationSpace=BoneRelative
    alphaBoolBlend {blendOption=Linear}
    boundBone {boneName=dyn_charm_01}
    componentAppliedLinearAccClamp=<3000 3000 3000>
    componentLinearAccScale=<1 1 1>
    componentLinearVelScale=<1 1 1>
    gravityOverride {z=-980}
    physicsBodyDefinitions [{
      boundBone {boneName=dyn_charm_01}
      boxExtents=<5 5 5>
      constraintSetup {angularLimitsMax=<5 30 30>, angularLimitsMin=<-5 -30 -30>, angularTarget=<1 0 0>}
      localJointOffset=<0.8 0 0>
     }]
    relativeSpaceBone {boneName=root}
    simSpaceSettings {maxAngularAcceleration=5000, maxAngularVelocity=5000}
   }
   n21 = Anim Dynamics - Bone: dyn_thigh_strap_01_r!(ComponentPose=@n20, bAlphaBoolEnabled:bool=True, ExternalForce="0,0,0") -> (Pose) {
    alphaInputType=Bool
    angularDampingOverride=0.99
    angularSpringConstant=300
    bAngularSpring=true
    bOverrideAngularDamping=true
    bUseGravityOverride=true
    simulationSpace=BoneRelative
    alphaBoolBlend {blendOption=Linear}
    boundBone {boneName=dyn_thigh_strap_01_r}
    componentAppliedLinearAccClamp=<3000 3000 3000>
    componentLinearAccScale=<0.7 0.7 0.7>
    componentLinearVelScale=<0.7 0.7 0.7>
    gravityOverride {z=-980}
    physicsBodyDefinitions [{
      boundBone {boneName=dyn_thigh_strap_01_r}
      boxExtents=<10 10 10>
      constraintSetup {coneAngle=1, angularLimitsMax=<5 20 10>, angularLimitsMin=<-5 0 -10>, angularTarget=<1 0 0>}
      localJointOffset=<1 0 0>
     }]
    relativeSpaceBone {boneName=root}
   }
   n22 = Anim Dynamics - Bone: dyn_strapBack_01_r!(ComponentPose=@n21, bAlphaBoolEnabled:bool=True, ExternalForce="0,0,0") -> (Pose) {
    alphaInputType=Bool
    angularDampingOverride=0.99
    angularSpringConstant=300
    bAngularSpring=true
    bOverrideAngularDamping=true
    bUseGravityOverride=true
    simulationSpace=BoneRelative
    alphaBoolBlend {blendOption=Linear}
    boundBone {boneName=dyn_strapBack_01_r}
    componentAppliedLinearAccClamp=<3000 3000 3000>
    componentLinearAccScale=<0.7 0.7 0.7>
    componentLinearVelScale=<0.7 0.7 0.7>
    gravityOverride {z=-980}
    physicsBodyDefinitions [{
      boundBone {boneName=dyn_strapBack_01_r}
      boxExtents=<10 10 10>
      constraintSetup {coneAngle=1, angularLimitsMax=<5 -10 0>, angularLimitsMin=<-5 -10 -20>, angularTarget=<1 0 0>}
      localJointOffset=<1 0 0>
     }]
    relativeSpaceBone {boneName=root}
   }
   n23 = Anim Dynamics - Bone: dyn_strapFront_01_r!(ComponentPose=@n22, bAlphaBoolEnabled:bool=True, ExternalForce="0,0,0") -> (Pose) {
    alphaInputType=Bool
    angularDampingOverride=0.99
    angularSpringConstant=300
    bAngularSpring=true
    bOverrideAngularDamping=true
    bUseGravityOverride=true
    simulationSpace=BoneRelative
    alphaBoolBlend {blendOption=Linear}
    boundBone {boneName=dyn_strapFront_01_r}
    componentAppliedLinearAccClamp=<3000 3000 3000>
    componentLinearAccScale=<0.7 0.7 0.7>
    componentLinearVelScale=<0.7 0.7 0.7>
    gravityOverride {z=-980}
    physicsBodyDefinitions [{
      boundBone {boneName=dyn_strapFront_01_r}
      boxExtents=<10 10 10>
      constraintSetup {coneAngle=1, angularLimitsMax=<5 20 10>, angularLimitsMin=<-5 0 0>, angularTarget=<1 0 0>}
      localJointOffset=<1 0 0>
     }]
    relativeSpaceBone {boneName=root}
   }
   n24 = Anim Dynamics - Bone: dyn_thigh_strap_01_l!(ComponentPose=@n23, bAlphaBoolEnabled:bool=True, ExternalForce="0,0,0") -> (Pose) {
    alphaInputType=Bool
    angularDampingOverride=0.9999
    angularSpringConstant=300
    bAngularSpring=true
    bOverrideAngularDamping=true
    bUseGravityOverride=true
    simulationSpace=BoneRelative
    alphaBoolBlend {blendOption=Linear}
    boundBone {boneName=dyn_thigh_strap_01_l}
    componentAppliedLinearAccClamp=<3000 3000 3000>
    componentLinearAccScale=<0.7 0.7 0.7>
    componentLinearVelScale=<0.7 0.7 0.7>
    gravityOverride {z=-980}
    physicsBodyDefinitions [{
      boundBone {boneName=dyn_thigh_strap_01_l}
      boxExtents=<10 10 10>
      constraintSetup {coneAngle=1, angularLimitsMax=<5 0 10>, angularLimitsMin=<-5 -70 -10>, angularTarget=<1 0 0>}
      localJointOffset=<1 0 0>
     }]
    relativeSpaceBone {boneName=root}
   }
   n25 = Anim Dynamics - Bone: dyn_strapFront_01_l!(ComponentPose=@n24, bAlphaBoolEnabled:bool=True, ExternalForce="0,0,0") -> (Pose) {
    alphaInputType=Bool
    angularDampingOverride=1
    angularSpringConstant=300
    bAngularSpring=true
    bOverrideAngularDamping=true
    bUseGravityOverride=true
    simulationSpace=BoneRelative
    alphaBoolBlend {blendOption=Linear}
    boundBone {boneName=dyn_strapFront_01_l}
    componentAppliedLinearAccClamp=<2000 2000 2000>
    componentLinearAccScale=<0.5 0.5 0.5>
    componentLinearVelScale=<0.5 0.5 0.5>
    gravityOverride {z=-980}
    physicsBodyDefinitions [{
      boundBone {boneName=dyn_strapFront_01_l}
      boxExtents=<10 10 10>
      constraintSetup {coneAngle=1, angularLimitsMax=<5 0 2>, angularLimitsMin=<-5 -20 -10>, angularTarget=<1 0 0>}
      localJointOffset=<1 0 0>
     }]
    relativeSpaceBone {boneName=root}
   }
   n26 = Anim Dynamics - Bone: dyn_strapFront_02_l!(ComponentPose=@n25, bAlphaBoolEnabled:bool=True, ExternalForce="0,0,0") -> (Pose) {
    alphaInputType=Bool
    angularDampingOverride=0.9999
    angularSpringConstant=300
    bAngularSpring=true
    bOverrideAngularDamping=true
    bUseGravityOverride=true
    simulationSpace=BoneRelative
    alphaBoolBlend {blendOption=Linear}
    boundBone {boneName=dyn_strapFront_02_l}
    componentAppliedLinearAccClamp=<2500 2500 2500>
    componentLinearAccScale=<0.6 0.6 0.6>
    componentLinearVelScale=<0.6 0.6 0.6>
    gravityOverride {z=-980}
    physicsBodyDefinitions [{
      boundBone {boneName=dyn_strapFront_02_l}
      boxExtents=<10 10 10>
      constraintSetup {coneAngle=1, angularLimitsMax=<5 5 0>, angularLimitsMin=<-5 -20 -20>, angularTarget=<1 0 0>}
      localJointOffset=<1 0 0>
     }]
    relativeSpaceBone {boneName=root}
   }
   n27 = Anim Dynamics - Bone: dyn_strapBack_01_l!(ComponentPose=@n26, bAlphaBoolEnabled:bool=True, ExternalForce="0,0,0") -> (Pose) {
    alphaInputType=Bool
    angularDampingOverride=0.99
    angularSpringConstant=300
    bAngularSpring=true
    bOverrideAngularDamping=true
    bUseGravityOverride=true
    simulationSpace=BoneRelative
    alphaBoolBlend {blendOption=Linear}
    boundBone {boneName=dyn_strapBack_01_l}
    componentAppliedLinearAccClamp=<3000 3000 3000>
    componentLinearAccScale=<0.7 0.7 0.7>
    componentLinearVelScale=<0.7 0.7 0.7>
    gravityOverride {z=-980}
    physicsBodyDefinitions [{
      boundBone {boneName=dyn_strapBack_01_l}
      boxExtents=<10 10 10>
      constraintSetup {coneAngle=1, angularLimitsMax=<5 -10 0>, angularLimitsMin=<-5 -10 -20>, angularTarget=<1 0 0>}
      localJointOffset=<1 0 0>
     }]
    relativeSpaceBone {boneName=root}
   }
   Save cached pose 'toSwitch'(Pose=Component To Local(ComponentPose=@n27)) {cacheName=toSwitch}
  ]}
 EventGraph {nodes [
   Event BlueprintUpdateAnimation()
   > n1 = Sequence()
   > n2 = Gravity Override State Function DEPRECATED(
    CharacterPartAnimInstance=Self-Reference()
    GameTimeInSeconds=GetGameTimeInSeconds()
    GravityOverrideParams=Get Gravity Override Params()
    RBAN Params=Get RBAN Params()
   )
   > Set Gravity Override(Gravity Override=@n2)
   "Set !frontend"(execute=@n1.then_1, "!frontend":bool=NOT Boolean(A=Get bIsInFrontEnd())) -> (Output_Get:bool)
   n3 = Get WeaponCoreAnim!() -> (WeaponCoreAnim)
   > Set isAssaultRifle(isAssaultRifle:bool="Equal (Enum)"(A:EFortWeaponCoreAnimation Enum=@n3, B:EFortWeaponCoreAnimation Enum=AssaultRifle)) -> (Output_Get:bool)
   > Set isPistol(isPistol="Equal (Enum)"(A=@n3, B=Pistol))
   > Set isSMG(isSMG:bool="Equal (Enum)"(A=@n3, B=SMG_P90)) -> (Output_Get:bool)
  ]}
]
variables [
 "!frontend" {type=boolean}
 Gravity Override {type=Vector}
 Gravity Override Params {type=GravityOverrideParamsStruct}
 HeadLimitAlpha {type=number}
 isAssaultRifle {type=boolean}
 isPistol {type=boolean}
 isSMG {type=boolean}
 RBAN Params {
  type=RBANParamsStruct
  default {
   dBNO {simSpaceSettings {
     maxAngularAcceleration=1000
     maxAngularVelocity=3
     maxLinearAcceleration=300
     maxLinearVelocity=700
     velocityScaleZ=0
     externalLinearDragV=<0 0.5 0>
    }}
   dBNO_Carried {simSpaceSettings {
     maxAngularAcceleration=1000
     maxAngularVelocity=3
     maxLinearAcceleration=300
     maxLinearVelocity=700
     velocityScaleZ=0
     externalLinearDragV=<0 0.5 0>
    }}
   emoteOrMelee {simSpaceSettings {
     maxAngularAcceleration=1000
     maxAngularVelocity=3
     maxLinearAcceleration=300
     maxLinearVelocity=700
     velocityScaleZ=0
     externalLinearDragV=<0 0.5 0>
    }}
   front_End {simSpaceSettings {
     maxAngularAcceleration=1000
     maxAngularVelocity=3
     maxLinearAcceleration=300
     maxLinearVelocity=700
     velocityScaleZ=0
     externalLinearDragV=<0 0.5 0>
    }}
   onGround_Crouch_Moving {simSpaceSettings {
     maxAngularAcceleration=1000
     maxAngularVelocity=3
     maxLinearAcceleration=300
     maxLinearVelocity=700
     velocityScaleZ=0
     externalLinearDragV=<0 0.5 0>
    }}
   onGround_Falling {simSpaceSettings {
     maxAngularAcceleration=1000
     maxAngularVelocity=3
     maxLinearAcceleration=300
     maxLinearVelocity=700
     velocityScaleZ=0
     externalLinearDragV=<0 0.5 0>
    }}
   onGround_Moving {simSpaceSettings {
     maxAngularAcceleration=1000
     maxAngularVelocity=3
     maxLinearAcceleration=300
     maxLinearVelocity=700
     velocityScaleZ=0
     externalLinearDragV=<0 0.5 0>
    }}
   onGround_Standing {simSpaceSettings {
     maxAngularAcceleration=1000
     maxAngularVelocity=3
     maxLinearAcceleration=300
     maxLinearVelocity=700
     velocityScaleZ=0
     externalLinearDragV=<0 0.5 0>
    }}
   rail_Grinding {simSpaceSettings {
     maxAngularAcceleration=1000
     maxAngularVelocity=3
     maxLinearAcceleration=300
     maxLinearVelocity=700
     velocityScaleZ=0
     worldAlpha=0.1
     externalLinearDragV=<0 0.5 0>
    }}
   riding_Moving {simSpaceSettings {
     maxAngularAcceleration=1000
     maxAngularVelocity=3
     maxLinearAcceleration=300
     maxLinearVelocity=700
     velocityScaleZ=0
     externalLinearDragV=<0 0.5 0>
    }}
   riding_Moving_Targeting {simSpaceSettings {
     maxAngularAcceleration=1000
     maxAngularVelocity=3
     maxLinearAcceleration=300
     maxLinearVelocity=700
     velocityScaleZ=0
     externalLinearDragV=<0 0.5 0>
    }}
   riding_Standing {simSpaceSettings {
     maxAngularAcceleration=1000
     maxAngularVelocity=3
     maxLinearAcceleration=300
     maxLinearVelocity=700
     velocityScaleZ=0
     externalLinearDragV=<0 0.5 0>
    }}
   skydive_DivingDown {simSpaceSettings {
     maxAngularAcceleration=1000
     maxAngularVelocity=3
     maxLinearAcceleration=300
     maxLinearVelocity=700
     velocityScaleZ=0
     externalLinearDragV=<0 0.5 0>
    }}
   skydive_FloatingDown {
    componentLinearAccScale=<0 0.5 0>
    componentLinearVelScale=<0 0.5 0>
    simSpaceSettings {
     maxAngularAcceleration=1000
     maxAngularVelocity=3
     maxLinearAcceleration=300
     maxLinearVelocity=700
     velocityScaleZ=0
     externalLinearDragV=<0 0.5 0>
    }
   }
   skydive_GoingUp {simSpaceSettings {
     maxAngularAcceleration=1000
     maxAngularVelocity=3
     maxLinearAcceleration=300
     maxLinearVelocity=700
     velocityScaleZ=0
     externalLinearDragV=<0 0.5 0>
    }}
   skydive_Parachuting {
    componentLinearAccScale=<0 0.1 0.3>
    componentLinearVelScale=<0 0.1 0.3>
    simSpaceSettings {
     maxAngularAcceleration=1000
     maxAngularVelocity=3
     maxLinearAcceleration=300
     maxLinearVelocity=700
     velocityScaleZ=0
     externalLinearDragV=<0 0.5 0>
    }
   }
   skydive_Surfing {simSpaceSettings {
     maxAngularAcceleration=1000
     maxAngularVelocity=3
     maxLinearAcceleration=300
     maxLinearVelocity=700
     velocityScaleZ=0
     externalLinearDragV=<0 0.5 0>
    }}
   swimming {simSpaceSettings {
     maxAngularAcceleration=1000
     maxAngularVelocity=3
     maxLinearAcceleration=300
     maxLinearVelocity=700
     velocityScaleZ=0
     externalLinearDragV=<0 0.5 0>
    }}
   tactical_Sprint {simSpaceSettings {
     maxAngularAcceleration=1000
     maxAngularVelocity=3
     maxLinearAcceleration=300
     maxLinearVelocity=700
     velocityScaleZ=0
     externalLinearDragV=<0 0.5 0>
    }}
   ziplining {simSpaceSettings {
     maxAngularAcceleration=1000
     maxAngularVelocity=3
     maxLinearAcceleration=300
     maxLinearVelocity=700
     velocityScaleZ=0
     externalLinearDragV=<0 0.5 0>
    }}
  }
 }
]
