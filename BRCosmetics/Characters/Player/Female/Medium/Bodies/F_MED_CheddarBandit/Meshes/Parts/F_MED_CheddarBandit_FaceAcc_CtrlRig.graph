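# Control Rig for the CheddarBandit female character's face accessory that adjusts helmet hair bones as the head moves.
# A spherical pose reader measures the head's X axis with spine_05 as optional parent, with debug drawing on.
# Its output weights a Modify Transforms node that offsets hair base bones, including those of strands c and d.
# A second, full-weight Modify Transforms node with different offsets is unconnected and never executes.
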ControlRigBlueprint /BRCosmetics/Characters/Player/Female/Medium/Bodies/F_MED_CheddarBandit/Meshes/Parts/F_MED_CheddarBandit_FaceAcc_CtrlRig
graphs [
 RigVMFunctionLibrary {links [], nodes []}
 RigVMModel {nodes [
   RigUnit_BeginExecution = &RigUnit_BeginExecution()
   n1 = &RigUnit_SphericalPoseReader(
    ActiveRegionScaleFactors=(PositiveWidth=1.2,NegativeWidth=1,PositiveHeight=1.8,NegativeHeight=1.8)
    ActiveRegionSize=0.1
    Debug=(bDrawDebug=True,bDraw2D=False,bDrawLocalAxes=False,DebugScale=25,DebugSegments=20,DebugThickness=0.25)
    DriverAxis=(X=1,Y=0,Z=0)
    DriverItem=(Type=Bone,Name="head")
    FalloffRegionScaleFactors=(PositiveWidth=2.1,NegativeWidth=1,PositiveHeight=5,NegativeHeight=5)
    FalloffSize=0.2
    FlipHeightScaling=False
    FlipWidthScaling=False
    OptionalParentItem=(Type=Bone,Name="spine_05")
    RotationOffset=(X=-90,Y=0,Z=35)
    ExecutePin=@RigUnit_BeginExecution.ExecutePin
   )
   > &RigUnit_ModifyTransforms(
    ItemToModify="((Item=(Type=Bone,Name=\"dyn_helmet_hair_base_d\"),Transform=(Rotation=(X=0,Y=0,Z=0.5818,W=-0.8133),Translation=(X=-2.7346,Y=-2.6385,Z=0.8636),Scale3D=(X=1,Y=1,Z=1))),(Item=(Type=Bone,Name=\"dyn_helmet_hair_base_с\"),Transform=(Rotation=(X=-0.1756,Y=0.2892,Z=- ...(truncated)"
    WeightMinimum=0
    Weight=@n1.OutputParam
   )
   &RigUnit_ModifyTransforms(
    ExecutePin=()
    ItemToModify="((Item=(Type=Bone,Name=\"dyn_helmet_hair_base_d\"),Transform=(Rotation=(X=0,Y=0,Z=0,W=-1),Translation=(X=-4.3884,Y=-2.6385,Z=0.8636),Scale3D=(X=1,Y=1,Z=1))),(Item=(Type=Bone,Name=\"dyn_helmet_hair_base_с\"),Transform=(Rotation=(X=0.1756,Y=-0.2892,Z=0.4461,W=-0 ...(truncated)"
    Weight=1
    WeightMinimum=0
   )
  ]}
]
hierarchy [
 root {loc=<0 0 0>, rot=<0 0 0 1>}
 pelvis {parent=root, loc=<0 -0.5301 98.0571>, rot=<0.0127 0.707 -0.0127 -0.707>}
 spine_01 {parent=pelvis, loc=<6.1208 -0.9328 0>, rot=<0 0 0.034 -0.9994>}
 spine_02 {parent=spine_01, loc=<6.5996 -0.2095 0>, rot=<0 0 0.0238 -0.9997>}
 spine_03 {parent=spine_02, loc=<6.4914 -0.1237 0>, rot=<0 0 0.0149 -0.9999>}
 spine_04 {parent=spine_03, loc=<6.9598 0.0794 0>, rot=<0 0 -0.0132 -0.9999>}
 spine_05 {parent=spine_04, loc=<6.9987 0 0>, rot=<0 0 -0.0015 -1>}
 neck_01 {parent=spine_05, loc=<11.3138 0.1529 0>, rot=<0 0 -0.2003 -0.9797>}
 neck_02 {parent=neck_01, loc=<4.5627 0 0>, rot=<0 0 0.0504 -0.9987>}
 head {parent=neck_02, loc=<5.3914 0 0>, rot=<0 0 0.115 -0.9934>}
 helmet_base {parent=head, loc=<0 0 0>, rot=<0 0 0 1>}
 dyn_helmet_ear_base_l {parent=helmet_base, loc=<10.3518 1.737 -6.9265>, rot=<0.4374 -0.2932 0.2006 -0.8261>}
 dyn_helmet_ear_01_l {parent=dyn_helmet_ear_base_l, loc=<0 0 0>, rot=<0 0 0 1>}
 dyn_helmet_ear_base_r {parent=helmet_base, loc=<10.3518 1.737 6.9265>, rot=<0.8261 -0.2006 0.2932 -0.4374>}
 dyn_helmet_ear_01_r {parent=dyn_helmet_ear_base_r, loc=<0 0 0>, rot=<0 0 0 1>}
 dyn_helmet_hair_base_a {parent=helmet_base, loc=<-1.027 1.7942 -7.2806>, rot=<0.2517 0.6466 0.6585 0.2912>}
 dyn_helmet_hair_01_a {parent=dyn_helmet_hair_base_a, loc=<0 0 0>, rot=<0 0 0 1>}
 dyn_helmet_hair_base_b {parent=helmet_base, loc=<-4.2378 -1.4023 -7.4483>, rot=<0.4777 0.5184 0.6923 0.1542>}
 dyn_helmet_hair_01_b {parent=dyn_helmet_hair_base_b, loc=<0 0 0>, rot=<0 0 0 1>}
 dyn_helmet_hair_base_с {parent=helmet_base, loc=<-3.5727 -6.2227 -3.0098>, rot=<0.0878 -0.9334 -0.3463 -0.0352>}
 dyn_helmet_hair_01_с {parent=dyn_helmet_hair_base_с, loc=<0 0 0>, rot=<0 0 0 1>}
 dyn_helmet_hair_base_d {parent=helmet_base, loc=<-4.0852 -5.8066 2.729>, rot=<0.0823 -0.9256 0.3649 0.0577>}
 dyn_helmet_hair_01_d {parent=dyn_helmet_hair_base_d, loc=<0 0 0>, rot=<0 0 0 1>}
 dyn_helmet_hair_base_e {parent=helmet_base, loc=<-2.6142 -2.8598 6.2365>, rot=<-0.1171 -0.8242 0.5477 0.0837>}
 dyn_helmet_hair_01_e {parent=dyn_helmet_hair_base_e, loc=<0 0 0>, rot=<0 0 0 1>}
 dyn_helmet_hair_base_f {parent=helmet_base, loc=<-1.5028 1.6445 6.7171>, rot=<-0.3401 -0.5725 0.6961 0.2683>}
 dyn_helmet_hair_01_f {parent=dyn_helmet_hair_base_f, loc=<0 0 0>, rot=<0 0 0 1>}
]
