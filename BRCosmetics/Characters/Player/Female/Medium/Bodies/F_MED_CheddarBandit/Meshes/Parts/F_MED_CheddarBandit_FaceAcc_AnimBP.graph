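# AnimBP for the CheddarBandit face accessory, adding secondary motion to helmet hair and ears on the character.
# It copies the attached parent's pose, runs a Control Rig, then chains AnimDynamics springs through the hair and both ears, with additive ear-base offsets.
# The event graph eases ear alphas from tactical sprint or targeting, but the AnimGraph never reads them.
# The head/neck limit bone chain and its alpha logic are disconnected.
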AnimBlueprint /BRCosmetics/Characters/Player/Female/Medium/Bodies/F_MED_CheddarBandit/Meshes/Parts/F_MED_CheddarBandit_FaceAcc_AnimBP
target_skeleton=&/F_MED_CheddarBandit_FaceAcc_Skeleton
graphs [
 AnimGraph {nodes [
   n1 = Get HeadLimitAlpha() -> (HeadLimitAlpha:double)
   n2 = "Transform (Modify) Bone - Bone: head"(Alpha=@n1, Translation="0,0,0", Rotation="0,0,35", Scale="1,1,1") {
    rotationMode=BMM_Additive
    scaleMode=BMM_Additive
    translationMode=BMM_Additive
    alphaBoolBlend {blendOption=Linear}
    boneToModify {boneName=head}
    rotation {roll=35}
   }
   n3 = "Transform (Modify) Bone - Bone: neck_02"(ComponentPose=@n2, Alpha=@n1, Translation="0,0,0", Rotation="0,0,10", Scale="1,1,1") {
    rotationMode=BMM_Additive
    scaleMode=BMM_Additive
    translationMode=BMM_Additive
    alphaBoolBlend {blendOption=Linear}
    boneToModify {boneName=neck_02}
    rotation {roll=10}
   }
   "Transform (Modify) Bone - Bone: neck_01"(ComponentPose=@n3, Alpha=@n1, Translation="0,0,0", Rotation="0,0,5", Scale="1,1,1") {
    rotationMode=BMM_Additive
    scaleMode=BMM_Additive
    translationMode=BMM_Additive
    alphaBoolBlend {blendOption=Linear}
    boneToModify {boneName=neck_01}
    rotation {roll=5}
   }
   n4 = Copy Pose From Mesh() {bUseAttachedParent=true}
   n5 = Control Rig(Source=Component To Local(ComponentPose=Local To Component(LocalPose=@n4))) {
    alphaBoolBlend {blendOption=Linear}
    controlRigAssetReference {blueprintRigClass=&/F_MED_CheddarBandit_FaceAcc_CtrlRig.F_MED_CheddarBandit_FaceAcc_CtrlRig_C}
   }
   n6 = Anim Dynamics - Bone: dyn_helmet_hair_01_a(ComponentPose=Local To Component(LocalPose=@n5), Alpha=1, ExternalForce="0,0,0") {
    angularBiasOverride=0.5
    angularDampingOverride=0.999
    angularSpringConstant=1500
    bAngularSpring=true
    bOverrideAngularDamping=true
    bUseGravityOverride=true
    simulationSpace=BoneRelative
    alphaBoolBlend {blendOption=Linear}
    boundBone {boneName=dyn_helmet_hair_01_a}
    componentAppliedLinearAccClamp=<5000 5000 5000>
    componentLinearAccScale=<1 1 1>
    componentLinearVelScale=<1 1 1>
    gravityOverride {z=-980}
    physicsBodyDefinitions [{
      boundBone {boneName=dyn_helmet_hair_01_a}
      boxExtents=<12 12 12>
      constraintSetup {angularLimitsMax=<5 10 3>, angularLimitsMin=<-5 -5 -15>, angularTarget=<1 0 0>}
      localJointOffset=<0.1 0 0>
     }]
    relativeSpaceBone {boneName=root}
    simSpaceSettings {maxAngularAcceleration=5000, maxAngularVelocity=5000, simSpaceAngularAlpha=1}
   }
   n7 = Anim Dynamics - Bone: dyn_helmet_hair_01_b(ComponentPose=@n6, Alpha=1, ExternalForce="0,0,0") {
    angularBiasOverride=0.5
    angularDampingOverride=0.999
    angularSpringConstant=1500
    bAngularSpring=true
    bOverrideAngularDamping=true
    bUseGravityOverride=true
    simulationSpace=BoneRelative
    alphaBoolBlend {blendOption=Linear}
    boundBone {boneName=dyn_helmet_hair_01_b}
    componentAppliedLinearAccClamp=<5000 5000 5000>
    componentLinearAccScale=<1 1 1>
    componentLinearVelScale=<1 1 1>
    gravityOverride {z=-980}
    physicsBodyDefinitions [{
      boundBone {boneName=dyn_helmet_hair_01_b}
      boxExtents=<12 12 12>
      constraintSetup {angularLimitsMax=<1 10 1>, angularLimitsMin=<-1 -10 -10>, angularTarget=<1 0 0>}
      localJointOffset=<0.1 0 0>
     }]
    relativeSpaceBone {boneName=root}
    simSpaceSettings {maxAngularAcceleration=5000, maxAngularVelocity=5000, simSpaceAngularAlpha=1}
   }
   n8 = Anim Dynamics - Bone: dyn_helmet_hair_01_с(ComponentPose=@n7, Alpha=1, ExternalForce="0,0,0") {
    angularBiasOverride=0.5
    angularDampingOverride=0.999
    angularSpringConstant=1500
    bAngularSpring=true
    bOverrideAngularDamping=true
    bUseGravityOverride=true
    simulationSpace=BoneRelative
    alphaBoolBlend {blendOption=Linear}
    boundBone {boneName=dyn_helmet_hair_01_с}
    componentAppliedLinearAccClamp=<5000 5000 5000>
    componentLinearAccScale=<1 1 1>
    componentLinearVelScale=<1 1 1>
    gravityOverride {z=-980}
    physicsBodyDefinitions [{
      boundBone {boneName=dyn_helmet_hair_01_с}
      boxExtents=<12 12 12>
      constraintSetup {angularLimitsMax=<1 10 10>, angularLimitsMin=<-1 -10 -10>, angularTarget=<1 0 0>}
      localJointOffset=<0.1 0 0>
     }]
    relativeSpaceBone {boneName=root}
    simSpaceSettings {maxAngularAcceleration=5000, maxAngularVelocity=5000, simSpaceAngularAlpha=1}
   }
   n9 = Anim Dynamics - Bone: dyn_helmet_hair_01_d(ComponentPose=@n8, Alpha=1, ExternalForce="0,0,0") {
    angularBiasOverride=0.5
    angularDampingOverride=0.999
    angularSpringConstant=1500
    bAngularSpring=true
    bOverrideAngularDamping=true
    bUseGravityOverride=true
    simulationSpace=BoneRelative
    alphaBoolBlend {blendOption=Linear}
    boundBone {boneName=dyn_helmet_hair_01_d}
    componentAppliedLinearAccClamp=<5000 5000 5000>
    componentLinearAccScale=<1 1 1>
    componentLinearVelScale=<1 1 1>
    gravityOverride {z=-980}
    physicsBodyDefinitions [{
      boundBone {boneName=dyn_helmet_hair_01_d}
      boxExtents=<12 12 12>
      constraintSetup {angularLimitsMax=<1 10 10>, angularLimitsMin=<-1 -10 -10>, angularTarget=<1 0 0>}
      localJointOffset=<0.1 0 0>
     }]
    relativeSpaceBone {boneName=root}
    simSpaceSettings {maxAngularAcceleration=5000, maxAngularVelocity=5000, simSpaceAngularAlpha=1}
   }
   n10 = Anim Dynamics - Bone: dyn_helmet_hair_01_e(ComponentPose=@n9, Alpha=1, ExternalForce="0,0,0") {
    angularBiasOverride=0.5
    angularDampingOverride=0.999
    angularSpringConstant=1500
    bAngularSpring=true
    bOverrideAngularDamping=true
    bUseGravityOverride=true
    simulationSpace=BoneRelative
    alphaBoolBlend {blendOption=Linear}
    boundBone {boneName=dyn_helmet_hair_01_e}
    componentAppliedLinearAccClamp=<5000 5000 5000>
    componentLinearAccScale=<1 1 1>
    componentLinearVelScale=<1 1 1>
    gravityOverride {z=-980}
    physicsBodyDefinitions [{
      boundBone {boneName=dyn_helmet_hair_01_e}
      boxExtents=<12 12 12>
      constraintSetup {angularLimitsMax=<1 10 10>, angularLimitsMin=<-1 -10 -10>, angularTarget=<1 0 0>}
      localJointOffset=<0.1 0 0>
     }]
    relativeSpaceBone {boneName=root}
    simSpaceSettings {maxAngularAcceleration=5000, maxAngularVelocity=5000, simSpaceAngularAlpha=1}
   }
   n11 = Anim Dynamics - Bone: dyn_helmet_hair_01_f(ComponentPose=@n10, Alpha=1, ExternalForce="0,0,0") {
    angularBiasOverride=0.5
    angularDampingOverride=0.999
    angularSpringConstant=1500
    bAngularSpring=true
    bOverrideAngularDamping=true
    bUseGravityOverride=true
    simulationSpace=BoneRelative
    alphaBoolBlend {blendOption=Linear}
    boundBone {boneName=dyn_helmet_hair_01_f}
    componentAppliedLinearAccClamp=<5000 5000 5000>
    componentLinearAccScale=<1 1 1>
    componentLinearVelScale=<1 1 1>
    gravityOverride {z=-980}
    physicsBodyDefinitions [{
      boundBone {boneName=dyn_helmet_hair_01_f}
      boxExtents=<12 12 12>
      constraintSetup {angularLimitsMax=<5 5 3>, angularLimitsMin=<-5 -10 -15>, angularTarget=<1 0 0>}
      localJointOffset=<0.1 0 0>
     }]
    relativeSpaceBone {boneName=root}
    simSpaceSettings {maxAngularAcceleration=5000, maxAngularVelocity=5000, simSpaceAngularAlpha=1}
   }
   n12 = "Transform (Modify) Bone - Bone: dyn_helmet_ear_base_l"(ComponentPose=@n11, Alpha=1, Translation="0,0,0", Rotation="8,-16,-6", Scale="1,1,1") {rotationMode=BMM_Additive, alphaBoolBlend {blendOption=Linear}, boneToModify {boneName=dyn_helmet_ear_base_l}, rotation=<rot 8 -16 -6>}
   n13 = Anim Dynamics - Bone: dyn_helmet_ear_01_l(ComponentPose=@n12, Alpha=1, ExternalForce="0,0,0") {
    angularBiasOverride=0.5
    angularDampingOverride=1
    angularSpringConstant=1500
    bAngularSpring=true
    bOverrideAngularDamping=true
    bUseGravityOverride=true
    simulationSpace=BoneRelative
    alphaBoolBlend {blendOption=Linear}
    boundBone {boneName=dyn_helmet_ear_01_l}
    componentAppliedLinearAccClamp=<5000 5000 5000>
    componentLinearAccScale=<1 1 1>
    componentLinearVelScale=<1 1 1>
    gravityOverride {z=-980}
    physicsBodyDefinitions [{
      boundBone {boneName=dyn_helmet_ear_01_l}
      boxExtents=<15 15 15>
      constraintSetup {angularLimitsMax=<0 0 10>, angularLimitsMin=<0 0 -10>, angularTarget=<1 0 0>}
      localJointOffset=<0.1 0 0>
     }]
    relativeSpaceBone {boneName=head}
    simSpaceSettings {maxAngularAcceleration=5000, maxAngularVelocity=5000, simSpaceAngularAlpha=1}
   }
   n14 = "Transform (Modify) Bone - Bone: dyn_helmet_ear_base_r"(ComponentPose=@n13, Alpha=1, Translation="0,0,0", Rotation="-8,16,-6", Scale="1,1,1") {rotationMode=BMM_Additive, alphaBoolBlend {blendOption=Linear}, boneToModify {boneName=dyn_helmet_ear_base_r}, rotation=<rot -8 16 -6>}
   n15 = Anim Dynamics - Bone: dyn_helmet_ear_01_r(ComponentPose=@n14, Alpha=1, ExternalForce="0,0,0") {
    angularBiasOverride=0.5
    angularDampingOverride=1
    angularSpringConstant=1500
    bAngularSpring=true
    bOverrideAngularDamping=true
    bUseGravityOverride=true
    simulationSpace=BoneRelative
    alphaBoolBlend {blendOption=Linear}
    boundBone {boneName=dyn_helmet_ear_01_r}
    componentAppliedLinearAccClamp=<5000 5000 5000>
    componentLinearAccScale=<1 1 1>
    componentLinearVelScale=<1 1 1>
    gravityOverride {z=-980}
    physicsBodyDefinitions [{
      boundBone {boneName=dyn_helmet_ear_01_r}
      boxExtents=<15 15 15>
      constraintSetup {angularLimitsMax=<0 0 10>, angularLimitsMin=<0 0 -10>, angularTarget=<1 0 0>}
      localJointOffset=<0.1 0 0>
     }]
    relativeSpaceBone {boneName=head}
    simSpaceSettings {maxAngularAcceleration=5000, maxAngularVelocity=5000, simSpaceAngularAlpha=1}
   }
   Output Pose(Result=Component To Local(ComponentPose=@n15)) {layerGroup=None}
  ]}
 EventGraph {nodes [
   n1 = Event BlueprintUpdateAnimation()
   n2 = OR Boolean(A=Get bIsTacticalSprinting(), B=Get bIsTargeting())
   > n3 = Set EarsBendBackAlpha(EarsBendBackAlpha:double=FInterpTo(
     Current=Get EarsBendBackAlpha() -> (EarsBendBackAlpha:double)
     Target="To Float (Boolean)"(InBool=@n2)
     DeltaTime=@n1.DeltaTimeX
     InterpSpeed=10
    )) -> (Output_Get:double)
   > Set EarsDynamicsAlpha(EarsDynamicsAlpha:double=Remap Float(Input=@n3, OldRange=(X=0,Y=1), NewRange=(X=1,Y=0), Clamp=true) -> (Output=~)) -> (Output_Get:double)
   Reroute Node()
   > n4 = Branch(Condition=Get bIsPlayingEmote())
   > Set HeadLimitAlpha(HeadLimitAlpha:double=0) -> (Output_Get:double)
   n5 = BreakTransform(InTransform=GetDeltaTransformFromRefPose(self=GetOwnerSkeletalMeshComponent(), BoneName=head, BaseName=neck_02))
   Set HeadLimitAlpha(execute=@n4.else, HeadLimitAlpha=MapRangeClamped(Value=BreakRotator(InRot=@n5.Rotation).Yaw, InRangeA=-10, InRangeB=-40))
  ]}
]
variables [
 EarsBendBackAlpha {type=number}
 EarsDynamicsAlpha {type=number}
 HeadLimitAlpha {type=number}
 Is Aim {type=boolean}
]
